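; BPI_PartPickerEntry is a Blueprint interface for entries in the Universal Part Picker widget of Pilgrim Quickplay.
; It declares two functions with no logic: SetDifficulty takes a song difficulty enum defaulting to Easy, and SetInstrument takes a track type enum defaulting to Guitar.
; Implementers receive the difficulty and instrument through these calls, so callers need no knowledge of the concrete entry class.
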(fn SetDifficulty
  (params (Difficulty :type EPilgrimSongDifficulty%20Enum :default "DifficultyEasy")))

(fn SetInstrument
  (params (Instrument :type EPilgrimTrackType%20Enum :default "TrackGuitar")))
