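# Sound cue for the difficulty-select UI sound in Pilgrim Quickplay.
# A single wave player plays the DifficultySelect sound wave directly, with no randomization, mixing or modulation nodes.
# The cue lowers its volume multiplier and routes through the Sparks green room sound class borrowed from SparksMaps.
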SoundCue /PilgrimQuickplay/Audio/UI/SC_DifficultySelect_Cue
root=WavePlayer_0
nodes [{id=WavePlayer_0, properties {soundWaveAssetPtr=&/DifficultySelect}}]
properties {maxDistance=2097152, soundClassObject=/SparksMaps/Audio/GreenRoom/Sparks_GreenRm, volumeMultiplier=0.3}
