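# Sound cue for the UI sound of opening the parts and difficulty screen in Pilgrim Quickplay.
# A mixer layers the Fortnite item shop menu toggle sound with a PartsAndDifficultyWipeOpen sound, with the wipe at half the toggle's input level.
# The cue plays at reduced overall volume under the Sparks green room sound class.
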SoundCue /PilgrimQuickplay/Audio/UI/SC_PartsAndDifficulty_Open_Cue
root=Mixer_0
nodes [
 {id=Mixer_0, children [WavePlayer_0,WavePlayer_1], properties {inputVolume [1,0.5]}}
 {id=WavePlayer_0, properties {soundWaveAssetPtr=/Game/Sounds/Fort_UI_Sounds/FrontEnd/ItemShop/wav/Fort_UI_Shop_MenuToggle}}
 {id=WavePlayer_1, properties {soundWaveAssetPtr=&/PartsAndDifficultyWipeOpen}}
]
properties {maxDistance=2097152, soundClassObject=/SparksMaps/Audio/GreenRoom/Sparks_GreenRm, volumeMultiplier=0.35}
