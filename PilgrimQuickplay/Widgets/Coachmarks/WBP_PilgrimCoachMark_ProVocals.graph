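# Pilgrim quickplay coach mark warning Pro Vocals players with Voice Chat that they switch to Push-to-Talk and teammates will not hear singing.
# Built on FortSidebarOnboardTooltipWidget, it fills a UIKit coach mark in PreConstruct and activates it on show, using menu input.
# "Got it" or tooltip end notifies close, broadcasts OnProVocalsTooltipDismissed and deactivates both widgets.
# RefocusOnButton restores focus; SetArrowPosition picks the arrow side from entry count.
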WidgetBlueprint /PilgrimQuickplay/Widgets/Coachmarks/WBP_PilgrimCoachMark_ProVocals
graphs [
 BP_GetDesiredInputConfig {nodes [
   Get Desired Input Config() {is=title K2Node_FunctionEntry}
   > Return Node!(
    execute
    ReturnValue:UIInput Config Structure=Make UIInput Config!(
     bIgnoreMoveInput:bool=true
     bIgnoreLookInput:bool=true
     InputMode:ECommonInputMode Enum=Menu
     MouseCaptureMode:EMouseCaptureMode Enum=NoCapture
     MouseLockMode:EMouseLockMode Enum=DoNotLock
     bHideCursorDuringViewportCapture:bool=True
    ) -> (UIInputConfig:UIInput Config Structure) {is=title K2Node_MakeStruct}
   )
  ]}
 EventGraph {nodes [
   "Button Other Action Clicked (WBP_UIKit_CoachMark)"() {is=title K2Node_ComponentBoundEvent}
   > n1 = HandleOtherActionClicked(self:Fort Sidebar Onboard Tooltip Widget Object Reference)
   NotifyCloseAnimationFinished(execute=[@n1.then, @n3.then], self:Fort Sidebar Onboard Tooltip Widget Object Reference)
   > Call On Pro Vocals Tooltip Dismissed!(execute:Exec, self:WBP Pilgrim Coach Mark Pro Vocals Object Reference) -> (then:Exec) {is=title K2Node_CallDelegate}
   > DeactivateWidget(self=Get WBP_UIKit_CoachMark() -> (WBP_UIKit_CoachMark:WBP UIKit Coach Mark Object Reference))
   > DeactivateWidget()
   RefocusOnButton() {is=title K2Node_CustomEvent}
   n2 = Get WBP_UIKit_CoachMark()
   > SetDesiredFocusWidget!(execute, self:User Widget Object Reference, Widget:Widget Object Reference=@n2) -> (then, ReturnValue:bool)
   > RequestRefreshFocus(self:Common Activatable Widget Object Reference=@n2)
   Event OnShowTooltip() {is=title K2Node_Event}
   > ActivateWidget(self=Get WBP_UIKit_CoachMark())
   n3 = Event OnEndTooltip() {is=title K2Node_Event}
   Event PreConstruct()
   n4 = Get WBP_UIKit_CoachMark()
   > SetTextHeader!(
    execute
    self:WBP UIKit Coach Mark Base Object Reference=@n4
    Text:Text=Format Text!(
     Format="LOCTEXT(\"Playing {provocals} with Voice Chat enabled?\")"
     provocals:Text=MakeLiteralText(Value="LOCTABLE(\"/SparksSettings/UI/ST_Sparks_InstrumentNames\", \"ProVocals\")")
    ) -> (Result)
   ) -> (then, Empty:bool)
   > SetTextDescription!(
    execute
    self:WBP UIKit Coach Mark Base Object Reference=@n4
    Text:Text="LOCTEXT(\"You\\'ll be switched to Push-to-Talk during gameplay. Your team will not hear you sing.\")"
   ) -> (then, Empty:bool)
  ]}
 OnProVocalsTooltipDismissed {nodes [OnProVocalsTooltipDismissed() {is=title K2Node_FunctionEntry}]}
 SetArrowPosition {nodes [
   n1 = SetArrowPosition!() -> (then, NumberEntries:int) {is=title K2Node_FunctionEntry}
   > n2 = Branch(Condition="integer > integer"(A=@n1, B=5))
   n3 = Get WBP_UIKit_CoachMark()
   > SetArrowPosition!(execute, self:WBP UIKit Coach Mark Base Object Reference=@n3, ArrowPosition:E_UI_OutBoardPosition Enum=NewEnumerator8) -> (then) {is=title K2Node_CallFunction}
   SetArrowPosition(execute=@n2.else, self=@n3, ArrowPosition=NewEnumerator9)
  ]}
]
variables [
 Arrow Position {type=E_UI_OutBoardPosition, default=Left_Down}
 OnProVocalsTooltipDismissed {}
]
widgets {
 info {
  inheritedWidgetCount=2
  parentClass=/Script/FortniteUI.FortSidebarOnboardTooltipWidget
  rootWidgetClass=/FortUILibrary/UIKit/CoachMark/WBP_UIKit_CoachMark.WBP_UIKit_CoachMark_C
  widgetCount=3
 }
 widgets [
  WBP_UIKit_CoachMark {
   widgetClassPath=/FortUILibrary/UIKit/CoachMark/WBP_UIKit_CoachMark
   bIsVariable=true
   widget=WBP_UIKit_CoachMark
   properties {
    arrowPosition=Left_Up
    bIsFocusable=true
    navigation=&self:WidgetTree.WBP_UIKit_CoachMark.WidgetNavigation_0
    shouldShowDismissButton=false
    textOtherActionButton=Got it
    widget Width=620
   }
  }
  Dismiss_Button {widgetClassPath=&CommonButtonLegacy, bInherited=true, widget=null}
  OtherAction_Button {widgetClassPath=&CommonButtonLegacy, bInherited=true, widget=null}
 ]
}
