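# Action-bar button widget for the Pilgrim Quickplay part picker, pairing an input-action icon with a text label such as "Change part/difficulty".
# SetInfo feeds a data table action row to the CommonActionWidget and sets the label text.
# On construct and on input method changes, CheckVisibilityBasedOnInput runs with the current input type.
# In compact UI only, it hides the label, not the icon, for touch and shows it as hit-test-invisible otherwise.
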WidgetBlueprint /PilgrimQuickplay/Widgets/UniversalPartPicker/WBP_PilgrimQuickplay_PartPicker_ActionBar
graphs [
 CheckVisibilityBasedOnInput {nodes [
   n1 = CheckVisibilityBasedOnInput!() -> (then, CurrentInput:ECommonInputType Enum) {is=title K2Node_FunctionEntry}
   > Branch(Condition=IsUsingCompactUI())
   > n2 = Branch(Condition="Equal (Enum)"(A=Touch, B=@n1))
   n3 = Get TextAction() -> (TextAction:Text Block Object Reference)
   > SetVisibility(self=@n3, InVisibility=Hidden)
   SetVisibility(execute=@n2.else, self=@n3, InVisibility=HitTestInvisible)
  ]}
 EventGraph {nodes [
   Event Construct()
   n1 = GetOwningPlayer()
   > Is Valid(InputObject=@n1)
   > CheckVisibilityBasedOnInput!(
    execute
    self
    CurrentInput:ECommonInputType Enum=GetCurrentInputType(self=Get CommonInputSubsystem!(PlayerController:Player Controller Object Reference=@n1) -> (ReturnValue:Common Input Object Reference) {is=title K2Node_GetSubsystemFromPC})
   ) -> (then) {is=title K2Node_CallFunction}
   n2 = Event On Input Method Changed!() -> (OutputDelegate, then, CurrentInputType:ECommonInputType Enum) {is=title K2Node_Event}
   > CheckVisibilityBasedOnInput(CurrentInput=@n2.CurrentInputType)
  ]}
 SetInfo {nodes [
   n1 = SetInfo!() -> (then, Text:Text, Action:Data Table Row Handle Structure) {is=title K2Node_FunctionEntry}
   > SetInputAction!(
    execute
    self:Common Action Widget Object Reference=Get CommonActionWidget() -> (CommonActionWidget:Common Action Widget Object Reference)
    InputActionRow:Data Table Row Handle Structure=@n1.Action
   ) -> (then)
   > Set Text(Text=@n1, self=Get TextAction())
  ]}
]
variables []
widgets {
 info {inheritedWidgetCount=1, parentClass=&CommonButtonBase, rootWidgetClass=&Overlay, widgetCount=6}
 widgets [
  Overlay_41 {widgetClassPath=&Overlay, widget=Overlay_41}
  HorizontalBox_47 {
   widgetClassPath=&HorizontalBox
   parent=Overlay_41
   widget=HorizontalBox_47
   slot {class=OverlaySlot, properties {horizontalAlignment=HAlign_Fill, verticalAlignment=VAlign_Fill, padding=<ltrb 0 0 0 5>}}
  }
  CommonActionWidget {
   widgetClassPath=&CommonActionWidget
   bIsVariable=true
   parent=HorizontalBox_47
   widget=CommonActionWidget
   properties {designTimeKey=Gamepad_DPad_Right}
   slot {class=HorizontalBoxSlot}
  }
  Spacer_66 {
   widgetClassPath=&Spacer
   parent=HorizontalBox_47
   widget=Spacer_66
   slot {class=HorizontalBoxSlot, properties {padding=<ltrb 10 0 0 0>}}
  }
  TextAction {
   widgetClassPath=&UEFN_TextBlock
   bIsVariable=true
   parent=HorizontalBox_47
   widget=TextAction
   properties {
    bScrollEnabled=true
    text=Change part/difficulty
    font {
     bForceMonospaced=false
     bMaterialIsStencil=false
     fontMaterial=null
     fontObject=&HeadingNow
     letterSpacing=0
     monospacedWidth=1
     size=16.5
     skewAmount=0
     typefaceFontName=75 Medium
     outlineSettings {
      bApplyOutlineToDropShadows=false
      bMiteredCorners=false
      bSeparateFillAlpha=false
      outlineMaterial=null
      outlineSize=0
      outlineColor=<rgba 0 0 0 1>
     }
    }
   }
   slot {class=HorizontalBoxSlot, properties {horizontalAlignment=HAlign_Left, verticalAlignment=VAlign_Center}}
  }
  InputActionWidget {widgetClassPath=&CommonActionWidget, bInherited=true, widget=null}
 ]
}
