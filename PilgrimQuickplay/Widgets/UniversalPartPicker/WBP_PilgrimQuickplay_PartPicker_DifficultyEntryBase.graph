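# CommonButton row for one song difficulty in the Pilgrim Quickplay part picker, showing a difficulty icon, name and optional slot widget.
# List item data fills the row, and click, Space or Enter broadcasts its difficulty through DifficultyEntryOnClicked.
# The slot widget resolves from a gameplay-tag alias, lazy-loads and receives difficulty by interface message.
# In-list entries highlight on hover or focus; compact UI adjusts height and gives standalone entries a chevron.
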WidgetBlueprint /PilgrimQuickplay/Widgets/UniversalPartPicker/WBP_PilgrimQuickplay_PartPicker_DifficultyEntryBase
graphs [
 DifficultyEntryOnClicked {nodes [DifficultyEntryOnClicked!() -> (then, EntryDifficulty:EPilgrimSongDifficulty Enum=DifficultyEasy) {is=title K2Node_FunctionEntry}]}
 EventGraph {nodes [
   AddWidgetAlias_LazyEntrySlot() {is=title K2Node_CustomEvent}
   > n1 = GetSoftWidgetClassForAlias!(
    execute
    self:Fort UIData Configuration Context Object Reference=Get FortUIDataConfigurationContext!() -> (ReturnValue:Fort UIData Configuration Context Object Reference) {is=title K2Node_GetBlueprintContextEx}
    WidgetAlias:"Gameplay Tag Structure (by ref)"=Get EntrySlotAliasTag() -> (EntrySlotAliasTag:Gameplay Tag Structure)
   ) -> (then, ReturnValue:User Widget Soft Class Reference)
   n2 = LazyWidgetCreated!() -> (OutputDelegate, then, Widget:User Widget Object Reference) {is=title K2Node_CustomEvent}
   SetLazyContentWithCallback!(
    execute=@n1.then
    self:Common Lazy Widget Object Reference=Get LazyWidget_EntrySlot() -> (LazyWidget_EntrySlot:Common Lazy Widget Object Reference)
    SoftWidget:User Widget Soft Class Reference=@n1
    OnCreatedCallback:"Delegate (by ref)"=@n2
   ) -> (then)
   n3 = OnPressedSelection() {is=title K2Node_CustomEvent}
   n4 = Cast To BP_PilgrimQuickplay_PartPicker_DifficultyEntryData!(
    execute=[@n3.then, @n17.then]
    Object=GetListItemObject!(UserObjectListEntry:User Object List Entry Interface) -> (ReturnValue:Object Reference)
   ) -> (
    then
    CastFailed
    AsBP Pilgrim Quickplay Part Picker Difficulty Entry Data:BP Pilgrim Quickplay Part Picker Difficulty Entry Data Object Reference
   ) {is=title K2Node_DynamicCast}
   > Call Difficulty Entry on Clicked!(
    execute:Exec
    self:WBP Pilgrim Quickplay Part Picker Difficulty Entry Base Object Reference
    EntryDifficulty:EPilgrimSongDifficulty Enum=Get DifficultyData!(self:BP Pilgrim Quickplay Part Picker Difficulty Entry Data Object Reference=@n4) -> (DifficultyData:EPilgrimSongDifficulty Enum=DifficultyEasy)
   ) -> (then:Exec) {is=title K2Node_CallDelegate}
   n5 = Event OnListItemObjectSet()
   > n6 = Cast To BP_PilgrimQuickplay_PartPicker_DifficultyEntryData(Object=@n5.ListItemObject)
   > Branch(Condition=Get bIsEntryOnList!(self:BP Pilgrim Quickplay Part Picker Difficulty Entry Data Object Reference=@n6) -> (bIsEntryOnList:bool))
   > SetDifficultyInfo!(
    execute
    self
    DifficultyData:EPilgrimSongDifficulty Enum=Get DifficultyData(self=@n6) -> (DifficultyData=DifficultyEasy)
    DifficultyText:Text=Get DifficultyText!(self:BP Pilgrim Quickplay Part Picker Difficulty Entry Data Object Reference=@n6) -> (DifficultyText:Text)
   ) -> (then)
   Event Construct()
   > n7 = Sequence()
   > n8 = Branch(Condition=Get bIsEntryOnList!() -> (bIsEntryOnList))
   > SetVisibility(self=Get ActionWidgetContainer() -> (ActionWidgetContainer:Overlay Object Reference), InVisibility=Collapsed)
   RemoveFromParent(execute=@n8.else, self=Get ActionWidgetContainer())
   SetMobileAdjustments(execute=@n7.then_1)
   n9 = Event PreConstruct()
   > n10 = Sequence()
   > n11 = GetDynamicMaterial(self=Get Image_Outline() -> (Image_Outline:Image Object Reference))
   > n12 = Set OutlineMat(OutlineMat:Material Instance Dynamic Object Reference=@n11) -> (Output_Get:Material Instance Dynamic Object Reference)
   > SetScalarParameterValue(self=@n12, ParameterName=CornerRadius, Value=0.5)
   > SetVisibility(self=Get Image_Outline(), InVisibility=Hidden)
   > SetVisibility(self=Get WBP_Sparks_BackingPanel() -> (WBP_Sparks_BackingPanel:WBP Sparks Backing Panel Object Reference), InVisibility=Hidden)
   > SetVisibility(self=Get LazyImage_RightEdge() -> (LazyImage_RightEdge:Fort Lazy Image Object Reference), InVisibility=Collapsed)
   > Branch(Condition=Get ShowEntrySlot() -> (ShowEntrySlot:bool))
   > SetVisibility(self=Get LazyWidget_EntrySlot(), InVisibility=SelfHitTestInvisible)
   > AddWidgetAlias_LazyEntrySlot() {is=title K2Node_CallFunction}
   Branch(execute=@n10.then_1, Condition=@n9.IsDesignTime)
   > Branch(Condition=Get bPreview_SeeMobileAdjustments() -> (bPreview_SeeMobileAdjustments:bool))
   > Set bIsEntryOnList(bIsEntryOnList:bool=false) -> (Output_Get:bool)
   > SetMobileAdjustments()
   Event On Hovered()
   > n13 = Branch(Condition=IsUsingInputType!(InputType:ECommonInputType Enum=MouseAndKeyboard, PlayerController:Player Controller Object Reference=GetOwningPlayer()) -> (ReturnValue:bool))
   Branch(execute=[@n15.then, @n13.then], Condition=Get bIsEntryOnList())
   > SetVisibility(self=Get ActionWidgetContainer(), InVisibility=HitTestInvisible)
   > SetVisibility(self=Get WBP_Sparks_BackingPanel(), InVisibility=HitTestInvisible)
   > SetVisibility(self=Get Image_Outline(), InVisibility=SelfHitTestInvisible)
   Event On Unhovered()
   > n14 = Branch(Condition=IsUsingInputType(InputType=MouseAndKeyboard, PlayerController=GetOwningPlayer()))
   Branch(execute=[@n16.then, @n14.then], Condition=Get bIsEntryOnList())
   > SetVisibility(self=Get ActionWidgetContainer(), InVisibility=Collapsed)
   > SetVisibility(self=Get WBP_Sparks_BackingPanel(), InVisibility=Hidden)
   > SetVisibility(self=Get Image_Outline(), InVisibility=Hidden)
   n15 = Event On Focused() {is=title K2Node_Event}
   n16 = Event On Unfocused() {is=title K2Node_Event}
   n17 = Event On Clicked()
   n18 = Cast To BP_PilgrimQuickplay_PartPicker_DifficultyEntryData(execute=@n19.then, Object=GetListItemObject())
   > SetDifficulty!(
    execute:Exec
    self:Object Reference=Get EntrySlot() -> (EntrySlot:User Widget Object Reference)
    Difficulty:EPilgrimSongDifficulty Enum=Get DifficultyData(self=@n18) -> (DifficultyData=DifficultyEasy)
   ) -> (then:Exec) {is=title K2Node_Message}
   n19 = Set EntrySlot(execute=@n2.then, EntrySlot:User Widget Object Reference=@n2.Widget) -> (Output_Get:User Widget Object Reference)
  ]}
 OnPreviewKeyDown {nodes [
   n1 = OnPreviewKeyDown!() -> (then, MyGeometry:Geometry Structure, InKeyEvent:Key Event Structure) {is=title K2Node_FunctionEntry}
   n2 = GetKey!(Input:"Key Event Structure (by ref)"=@n1.InKeyEvent) -> (ReturnValue:Key Structure)
   > n3 = Branch(Condition="Equal (Key)"(A:Key Structure=@n2, B:Key Structure=SpaceBar) -> (ReturnValue:bool=~) {is=title K2Node_PromotableOperator})
   OnPressedSelection(execute=[@n4.then, @n3.then]) {is=title K2Node_CallFunction}
   > Return Node!(execute, ReturnValue:Event Reply Structure=Handled())
   n4 = Branch(execute=@n3.else, Condition="Equal (Key)"(A=@n2, B=Enter) -> (ReturnValue=~))
   Return Node(execute=@n4.else, ReturnValue=Unhandled())
  ]}
 SetDifficultyInfo {nodes [
   n1 = SetDifficultyInfo!() -> (then, DifficultyData:EPilgrimSongDifficulty Enum=DifficultyEasy, DifficultyText:Text) {is=title K2Node_FunctionEntry}
   > SetDifficulty!(
    execute
    self:WBP Pilgrim Results Difficulty Object Reference=Get WBP_Pilgrim_Results_Difficulty() -> (WBP_Pilgrim_Results_Difficulty:WBP Pilgrim Results Difficulty Object Reference)
    Difficulty:EPilgrimSongDifficulty Enum=@n1
   ) -> (then) {is=title K2Node_CallFunction}
   > "SetText (Text)"(self=Get DifficultyNameLabel() -> (DifficultyNameLabel:Common Text Object Reference), InText=@n1.DifficultyText)
   > n2 = Get!(execute:Exec) -> (then:Exec, else:Exec, EntrySlot:User Widget Object Reference)
   > SetDifficulty(self=@n2, Difficulty=Get DifficultyData!() -> (DifficultyData=DifficultyEasy)) {is=title K2Node_Message}
  ]}
 SetMobileAdjustments {nodes [
   SetMobileAdjustments() {is=title K2Node_FunctionEntry}
   > Branch(Condition=IsUsingCompactUI())
   > n1 = Branch(Condition=Get bIsEntryOnList())
   > SetMinDesiredHeight!(execute, self:Size Box Object Reference=Get SizeBox_Entry() -> (SizeBox_Entry:Size Box Object Reference), InMinDesiredHeight:float=90) -> (then)
   SetVisibility(execute=@n1.else, self=Get LazyImage_RightEdge(), InVisibility=HitTestInvisible)
   > SetBrushFromLazyMaterial!(
    execute
    self:Common Lazy Image Object Reference=Get LazyImage_RightEdge()
    LazyMaterial:"Material Interface Soft Object Reference (by ref)"=Get ExpandIconMaterial() -> (ExpandIconMaterial:Material Interface Soft Object Reference)
   ) -> (then)
   > SetMinDesiredHeight(self=Get SizeBox_Entry(), InMinDesiredHeight=75)
  ]}
]
variables [
 bIsEntryOnList {type=boolean, default=true}
 bIsOnlyInfo {type=boolean}
 BP_PilgrimPlayerControllerComponent {type=BP_PilgrimPlayerControllerComponent_C, default=null}
 bPreview_SeeMobileAdjustments {type=boolean, default=true}
 DifficultyEntryOnClicked {}
 EntrySlot {type=UserWidget, default=null}
 EntrySlotAliasTag {type=GameplayTag, default {tagName=Pilgrim.Hud.PartPicker.Entry.Slot}}
 ExpandIconMaterial {type=MaterialInterface, default=/FortUILibrary/Textures/MaterialInstances/MI_UIKit_Icon_Chevron_Up}
 FilledMaterial {type=MaterialInterface, default=/PilgrimQuickplay/Widgets/Materials/MI_UI_Pilgrim_Intensity_Filled}
 InstrumentIconMap {default {
   bass=/PilgrimQuickplay/Widgets/Materials/MI_UI_Icon_Instrument_Type_Bass_Sm
   drum=/PilgrimQuickplay/Widgets/Materials/MI_UI_Icon_Instrument_Type_Drums_Sm
   guitar=/PilgrimQuickplay/Widgets/Materials/MI_UI_Icon_Instrument_Type_Guitar_Sm
   keyboard=/PilgrimQuickplay/Widgets/Materials/MI_UI_Icon_Instrument_Type_Keytar_Sm
   vocals=/PilgrimQuickplay/Widgets/Materials/MI_UI_Icon_Instrument_Type_Vocals_Sm
  }}
 InstrumentIndex {type=integer}
 isLarge {type=boolean}
 MaxIntensity {type=integer, default=7}
 OutlineMat {type=MaterialInstanceDynamic}
 ShowBackground {type=boolean}
 ShowEntrySlot {type=boolean, default=true}
 ShowMeterLabel {type=boolean}
 UnfilledMaterial {type=MaterialInterface, default=/PilgrimQuickplay/Widgets/Materials/MI_UI_Pilgrim_Intensity_Empty}
]
widgets {
 info {inheritedWidgetCount=1, parentClass=&CommonButtonBase, rootWidgetClass=&SizeBox, widgetCount=16}
 widgets [
  SizeBox_Entry {widgetClassPath=&SizeBox, bIsVariable=true, widget=SizeBox_Entry, properties {bOverride_WidthOverride=true, widthOverride=400}}
  Overlay_34 {widgetClassPath=&Overlay, parent=SizeBox_Entry, widget=Overlay_34, slot {class=SizeBoxSlot}}
  WBP_Sparks_BackingPanel {
   widgetClassPath=/PilgrimCore/Widgets/SparksUILibrary/WBP_Sparks_BackingPanel
   bIsVariable=true
   parent=Overlay_34
   widget=WBP_Sparks_BackingPanel
   properties {blurFrame=false, cornerRadius=XLarge, darkenLayer=false, eSparksBackingPanelType=Vertical Gradient}
   slot {class=OverlaySlot, properties {horizontalAlignment=HAlign_Fill, verticalAlignment=VAlign_Fill}}
  }
  Image_Outline {
   widgetClassPath=&Image
   bIsVariable=true
   parent=Overlay_34
   widget=Image_Outline
   properties {brush {
     bIsDynamicallyLoaded=false
     drawAs=Image
     imageType=FullColor
     mirroring=NoMirror
     resourceName=None
     resourceObject=/PilgrimCore/Widgets/Materials/MI_UI_Pilgrim_Rounded_Outline_SelectedLeadeboards
     tiling=NoTile
     imageSize=<32 32>
     margin=<ltrb 0 0 0 0>
     outlineSettings {
      bUseBrushTransparency=false
      roundingType=HalfHeightRadius
      width=0
      color {colorUseRule=UseColor_Specified, specifiedColor=<rgba 0 0 0 0>}
      cornerRadii=<0 0 0 0>
     }
     tintColor {colorUseRule=UseColor_Specified, specifiedColor=<rgba 1 1 1 1>}
     uVRegion {bIsValid=false, max=<0 0>, min=<0 0>}
    }}
   slot {class=OverlaySlot, properties {horizontalAlignment=HAlign_Fill, verticalAlignment=VAlign_Fill}}
  }
  EntryContainer {
   widgetClassPath=&StackBox
   parent=Overlay_34
   widget=EntryContainer
   properties {bOverride_Cursor=true}
   slot {class=OverlaySlot, properties {horizontalAlignment=HAlign_Fill, verticalAlignment=VAlign_Center, padding=<ltrb 16 10 16 10>}}
  }
  ActionWidgetContainer {
   widgetClassPath=&Overlay
   bIsVariable=true
   parent=EntryContainer
   widget=ActionWidgetContainer
   slot {class=StackBoxSlot, properties {horizontalAlignment=HAlign_Left, verticalAlignment=VAlign_Center, padding=<ltrb -10 0 4 0>}}
  }
  CommonActionWidget {
   widgetClassPath=&CommonActionWidget
   bIsVariable=true
   parent=ActionWidgetContainer
   widget=CommonActionWidget
   properties {
    inputActions [{dataTable=&InputActionDataTable, rowName=Confirm}]
    progressMaterialBrush {
     bIsDynamicallyLoaded=false
     drawAs=Image
     imageType=NoImage
     mirroring=NoMirror
     resourceName=None
     resourceObject=null
     tiling=NoTile
     imageSize=<32 32>
     margin=<ltrb 0 0 0 0>
     outlineSettings {
      bUseBrushTransparency=false
      roundingType=HalfHeightRadius
      width=0
      color {colorUseRule=UseColor_Specified, specifiedColor=<rgba 0 0 0 0>}
      cornerRadii=<0 0 0 0>
     }
     tintColor {colorUseRule=UseColor_Specified, specifiedColor=<rgba 1 1 1 0>}
     uVRegion {bIsValid=false, max=<0 0>, min=<0 0>}
    }
   }
   slot {class=OverlaySlot}
  }
  SizeBox_85 {
   widgetClassPath=&SizeBox
   parent=EntryContainer
   widget=SizeBox_85
   properties {bOverride_HeightOverride=true, bOverride_WidthOverride=true, heightOverride=40, maxDesiredWidth=40, widthOverride=40}
   slot {class=StackBoxSlot}
  }
  ScaleBox_0 {
   widgetClassPath=&ScaleBox
   parent=SizeBox_85
   widget=ScaleBox_0
   properties {stretch=UserSpecified, userSpecifiedScale=0.7}
   slot {class=SizeBoxSlot, properties {horizontalAlignment=HAlign_Center, verticalAlignment=VAlign_Center}}
  }
  WBP_Pilgrim_Results_Difficulty {
   widgetClassPath=/PilgrimQuickplay/Widgets/Results/WBP_Pilgrim_Results_Difficulty
   bIsVariable=true
   parent=ScaleBox_0
   widget=WBP_Pilgrim_Results_Difficulty
   slot {class=ScaleBoxSlot}
  }
  DifficultyNameLabel {
   widgetClassPath=&CommonTextBlock
   bIsVariable=true
   parent=EntryContainer
   widget=DifficultyNameLabel
   properties {
    justification=Center
    scrollStyle=/PilgrimQuickplay/Widgets/ScrollStyles/TextScrollStyle_PilgrimQuickplay_Setlist.TextScrollStyle_PilgrimQuickplay_Setlist_C
    text=Easy
    textTransformPolicy=ToUpper
    visibility=HitTestInvisible
    colorAndOpacity {colorUseRule=UseColor_Specified, specifiedColor=<rgba 0.6038 0.8879 1 1>}
    font {
     bForceMonospaced=false
     bMaterialIsStencil=false
     fontMaterial=null
     fontObject=&HeadingNow
     letterSpacing=0
     monospacedWidth=1
     size=16.5
     skewAmount=0
     typefaceFontName=76 Bold Italic
     outlineSettings {
      bApplyOutlineToDropShadows=false
      bMiteredCorners=false
      bSeparateFillAlpha=false
      outlineMaterial=null
      outlineSize=0
      outlineColor=<rgba 0 0 0 1>
     }
    }
    shadowOffset=<0 0>
   }
   slot {
    class=StackBoxSlot
    properties {horizontalAlignment=HAlign_Left, verticalAlignment=VAlign_Center, padding=<ltrb 8 0 0 0>, size {sizeRule=Fill, value=1}}
   }
  }
  Spacer_246 {widgetClassPath=&Spacer, parent=EntryContainer, widget=Spacer_246, properties {size=<20 1>}, slot {class=StackBoxSlot}}
  PartPickerEntrySlot {
   widgetClassPath=&Overlay
   bIsVariable=true
   parent=EntryContainer
   widget=PartPickerEntrySlot
   properties {visibility=Collapsed}
   slot {class=StackBoxSlot, properties {horizontalAlignment=HAlign_Right}}
  }
  LazyWidget_EntrySlot {
   widgetClassPath=/Script/CommonUI.CommonLazyWidget
   bIsVariable=true
   parent=EntryContainer
   widget=LazyWidget_EntrySlot
   properties {visibility=Collapsed}
   slot {class=StackBoxSlot, properties {horizontalAlignment=HAlign_Right}}
  }
  LazyImage_RightEdge {
   widgetClassPath=&FortLazyImage
   bIsVariable=true
   parent=EntryContainer
   widget=LazyImage_RightEdge
   properties {
    brush {
     bIsDynamicallyLoaded=false
     drawAs=Image
     imageType=NoImage
     mirroring=NoMirror
     resourceName=None
     resourceObject=null
     tiling=NoTile
     imageSize=<24 24>
     margin=<ltrb 0 0 0 0>
     outlineSettings {
      bUseBrushTransparency=false
      roundingType=HalfHeightRadius
      width=0
      color {colorUseRule=UseColor_Specified, specifiedColor=<rgba 0 0 0 0>}
      cornerRadii=<0 0 0 0>
     }
     tintColor {colorUseRule=UseColor_Specified, specifiedColor=<rgba 1 1 1 1>}
     uVRegion {bIsValid=false, max=<0 0>, min=<0 0>}
    }
    loadingBackgroundBrush {
     bIsDynamicallyLoaded=false
     drawAs=NoDrawType
     imageType=NoImage
     mirroring=NoMirror
     resourceName=None
     resourceObject=null
     tiling=NoTile
     imageSize=<24 24>
     margin=<ltrb 0 0 0 0>
     outlineSettings {
      bUseBrushTransparency=false
      roundingType=HalfHeightRadius
      width=0
      color {colorUseRule=UseColor_Specified, specifiedColor=<rgba 0 0 0 0>}
      cornerRadii=<0 0 0 0>
     }
     tintColor {colorUseRule=UseColor_Specified, specifiedColor=<rgba 1 1 1 1>}
     uVRegion {bIsValid=false, max=<0 0>, min=<0 0>}
    }
   }
   slot {class=StackBoxSlot, properties {horizontalAlignment=HAlign_Center, verticalAlignment=VAlign_Center}}
  }
  InputActionWidget {widgetClassPath=&CommonActionWidget, bInherited=true, widget=null}
 ]
}
