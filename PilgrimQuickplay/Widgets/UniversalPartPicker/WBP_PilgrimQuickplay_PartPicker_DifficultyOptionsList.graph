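# Activatable difficulty picker list for the Pilgrim Quickplay part picker, showing difficulty entries over a backing panel.
# On activation or entry generation it refreshes, navigates to the preselected index and refreshes focus.
# Entry clicks are relayed through a dispatcher, with a select sound pitched higher for harder difficulties.
# Back plays a sound and broadcasts closing but returns unhandled; focus-path loss can broadcast mobile dismissal, gated by IsPilgrimInverse.
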WidgetBlueprint /PilgrimQuickplay/Widgets/UniversalPartPicker/WBP_PilgrimQuickplay_PartPicker_DifficultyOptionsList
graphs [
 BP_GetDesiredFocusTarget {nodes [
   Get Desired Focus Target()
   > Return Node!(execute, ReturnValue:Widget Object Reference=Get CommonListView() -> (CommonListView:Common List View Object Reference))
  ]}
 BP_GetDesiredInputConfig {nodes [
   Get Desired Input Config() {is=title K2Node_FunctionEntry}
   > Return Node(ReturnValue:UIInput Config Structure=Make UIInput Config!(
     bIgnoreMoveInput:bool=true
     bIgnoreLookInput:bool=true
     InputMode:ECommonInputMode Enum=Menu
     MouseCaptureMode:EMouseCaptureMode Enum=NoCapture
     MouseLockMode:EMouseLockMode Enum=DoNotLock
     bHideCursorDuringViewportCapture:bool=True
    ) -> (UIInputConfig:UIInput Config Structure) {is=title K2Node_MakeStruct})
  ]}
 BP_OnHandleBackAction {nodes [
   On Handle Back Action() {is=title K2Node_FunctionEntry}
   > PlaySound2D(Sound=Get SoundBackAction() -> (SoundBackAction:Sound Base Object Reference))
   > Call On Closing Difficulty List!(execute:Exec, self:WBP Pilgrim Quickplay Part Picker Difficulty Options List Object Reference) -> (then:Exec) {is=title K2Node_CallDelegate}
   > Return Node(ReturnValue:bool=false)
  ]}
 EventGraph {nodes [
   n1 = "On Entries Generated (CommonListView)"!() -> (OutputDelegate, then, NumEntries:int=0) {is=title K2Node_ComponentBoundEvent}
   n2 = Get CommonListView()
   RequestRefresh(execute=[@n1.then, @n6.then], self:List View Base Object Reference=@n2)
   > NavigateToIndex!(execute, self:List View Object Reference=@n2, Index:int=Get PreselectedDifficultyIndex() -> (PreselectedDifficultyIndex:int=0)) -> (then)
   > RequestRefreshFocus(self:Common Activatable Widget Object Reference)
   n3 = "On Entry Initialized (CommonListView)"!() -> (OutputDelegate, then, Item:Object Reference, Widget:User Widget Object Reference) {is=title K2Node_ComponentBoundEvent}
   > n4 = Cast To WBP_PilgrimQuickplay_PartPicker_DifficultyEntryBase!(execute, Object=@n3.Widget) -> (
    then
    CastFailed
    AsWBP Pilgrim Quickplay Part Picker Difficulty Entry Base:WBP Pilgrim Quickplay Part Picker Difficulty Entry Base Object Reference
   ) {is=title K2Node_DynamicCast}
   n5 = OnDifficultySelected!() -> (OutputDelegate, then, EntryDifficulty:EPilgrimSongDifficulty Enum=DifficultyEasy) {is=title K2Node_CustomEvent}
   Bind Event to Difficulty Entry on Clicked(execute=@n4.then, self:WBP Pilgrim Quickplay Part Picker Difficulty Entry Base Object Reference=@n4, Delegate=@n5) {is=title K2Node_AddDelegate}
   Event OnRemovedFromFocusPath!() -> (OutputDelegate, then, InFocusEvent:Focus Event Structure) {is=title K2Node_Event}
   > Branch(Condition=IsPilgrimInverse!(OwningPlayerController:Player Controller Object Reference=GetOwningPlayer()) -> (IsPilgrimInverse:bool))
   > Call On Mobile Dismissed!(execute:Exec, self:WBP Pilgrim Quickplay Part Picker Difficulty Options List Object Reference) -> (then:Exec) {is=title K2Node_CallDelegate}
   n6 = Event On Activated()
   PlaySound2D(
    execute=@n7.then
    Sound=Get SoundEntrySelected() -> (SoundEntrySelected:Sound Base Object Reference)
    PitchMultiplier=Select!(
     DifficultyEasy:float=0.85
     DifficultyMedium:float=1
     DifficultyHard:float=1.15
     DifficultyExpert:float=1.3
     None:float=1
     Index:EPilgrimSongDifficulty Enum=@n5.EntryDifficulty
    ) -> (ReturnValue:float)
   )
   n7 = Call On Difficulty Entry Clicked!(
    execute:Exec=@n5.then
    self:WBP Pilgrim Quickplay Part Picker Difficulty Options List Object Reference
    Difficulty:EPilgrimSongDifficulty Enum=@n5.EntryDifficulty
   ) -> (then:Exec) {is=title K2Node_CallDelegate}
  ]}
 On Closing Difficulty List {nodes [On Closing Difficulty List() {is=title K2Node_FunctionEntry}]}
 On Difficulty Entry Clicked {nodes [On Difficulty Entry Clicked!() -> (then, Difficulty:EPilgrimSongDifficulty Enum=DifficultyEasy) {is=title K2Node_FunctionEntry}]}
 On Mobile Dismissed {nodes [On Mobile Dismissed() {is=title K2Node_FunctionEntry}]}
]
variables [
 ListDifficulties {type="array<EPilgrimSongDifficulty>", default []}
 On Closing Difficulty List {}
 On Difficulty Entry Clicked {}
 On Mobile Dismissed {}
 PreselectedDifficultyIndex {type=integer, default=1}
 SoundBackAction {type=SoundBase, default=/PilgrimQuickplay/Audio/UI/SC_PartsAndDifficulty_Close_Cue}
 SoundEntrySelected {type=SoundBase, default=/PilgrimQuickplay/Audio/UI/SC_DifficultySelect_Cue}
]
widgets {
 info {inheritedWidgetCount=0, parentClass=&CommonActivatableWidget, rootWidgetClass=&Overlay, widgetCount=3}
 widgets [
  Overlay_0 {widgetClassPath=&Overlay, widget=Overlay_0, properties {clipping=ClipToBounds}}
  WBP_Sparks_BackingPanel {
   widgetClassPath=/PilgrimCore/Widgets/SparksUILibrary/WBP_Sparks_BackingPanel
   bIsVariable=true
   parent=Overlay_0
   widget=WBP_Sparks_BackingPanel
   properties {blurFrame=false, cornerRadius=Small, eSparksBackingPanelType=Simple Color, visibility=HitTestInvisible}
   slot {class=OverlaySlot, properties {horizontalAlignment=HAlign_Fill, verticalAlignment=VAlign_Fill}}
  }
  CommonListView {
   widgetClassPath=&CommonListView
   bIsVariable=true
   parent=Overlay_0
   widget=CommonListView
   properties {
    bAllowKeepPreselectedItems=false
    clipping=Inherit
    entryWidgetClass=&/WBP_PilgrimQuickplay_PartPicker_DifficultyEntryBase.WBP_PilgrimQuickplay_PartPicker_DifficultyEntryBase_C
    navigation=&self:WidgetTree.CommonListView.WidgetNavigation_0
    numDesignerPreviewEntries=4
   }
   slot {class=OverlaySlot, properties {horizontalAlignment=HAlign_Fill, verticalAlignment=VAlign_Fill}}
  }
 ]
}
