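# CommonUI button for one instrument in the Pilgrim Quickplay part picker, broadcasting the chosen track type on click, Space or Enter.
# List item data drives its name, material-based icon, intensity pips and missing-parts viewmodel.
# A tag-aliased, lazy-loaded slot widget can replace the pips.
# Focus and hover reveal highlights and peripheral or Pro Vocals hints, with Pro Vocals also notifying the playspace event router; compact UI adjusts height and chevron.
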WidgetBlueprint /PilgrimQuickplay/Widgets/UniversalPartPicker/WBP_PilgrimQuickplay_PartPicker_InstrumentEntryBase
graphs [
 CheckProVocalsCase {nodes [
   CheckProVocalsCase() {is=title K2Node_FunctionEntry}
   > Branch(Condition="Equal (Enum)"(
     A:EPilgrimTrackType Enum=Get CurrentEntryInstrumentType() -> (CurrentEntryInstrumentType:EPilgrimTrackType Enum=TrackGuitar)
     B:EPilgrimTrackType Enum=TrackProVocals
    ))
   > InitializeBindings(self:MVVM View Object Reference=Get View From User Widget!(
     self:Viewmodel Engine Subsytem Object Reference=Get MVVMSubsystem() -> (ReturnValue:Viewmodel Engine Subsytem Object Reference) {is=title K2Node_GetEngineSubsystem}
     UserWidget:User Widget Object Reference=Self-Reference()
    ) -> (ReturnValue:MVVM View Object Reference) {is=title K2Node_CallFunction})
   > Set bShowInfoDetail(bShowInfoDetail:bool=true) -> (Output_Get:bool)
  ]}
 DetermineMicInfoText {nodes [
   n1 = DetermineMicInfoText!() -> (then, MicInfo:Text) {is=title K2Node_FunctionEntry}
   > n2 = Branch(Condition=IsUsingCompactUI())
   n3 = Get InfoText() -> (InfoText:Common Text Object Reference)
   > Set Text(Text="LOCTEXT(\"Headphones Recommended\")", self=@n3)
   Set Text(execute=@n2.else, Text=@n1, self=@n3)
  ]}
 DetermineShowProVocalsNotification {nodes [
   n1 = DetermineShowProVocalsNotification!() -> (then, bShowTopNotification:bool=~) {is=title K2Node_FunctionEntry}
   n2 = Get CurrentEntryInstrumentType() -> (CurrentEntryInstrumentType=TrackGuitar)
   > Branch(Condition=AND Boolean(A="Equal (Enum)"(A=@n2, B=TrackProVocals), B=Get bIsEntryOnList() -> (bIsEntryOnList:bool)))
   n3 = GetOwningPlayer()
   > n4 = GetPlayspaceForActor!(execute, WorldContextObject:Object Reference=@n3, InActor:Actor Object Reference=@n3) -> (then, ReturnValue:Playspace Object Reference)
   > n5 = GetEventRouter!(
    execute
    Scope:Object Class Reference=Get Class(Object=@n4) -> (ReturnValue:Playspace Class Reference)
    ContextActor:Actor Object Reference=@n3
   ) -> (then, ReturnValue:Gameplay Event Router Component Object Reference)
   > Broadcast Event!(
    execute
    self:Gameplay Event Router Component Object Reference=@n5
    EventData:Pilgrim Event Part Picker Pro Vocals Selected Structure=Make Pilgrim Event Part Picker Pro Vocals Selected!(bIsEntryFocused:bool=@n1) -> (PilgrimEvent_PartPickerProVocalsSelected:Pilgrim Event Part Picker Pro Vocals Selected Structure) {is=title K2Node_MakeStruct}
    EventContext:Object Reference=@n3
   ) -> (then, ReturnValue:bool)
  ]}
 EventGraph {nodes [
   AddWidgetAlias_PickerEntrySlot() {is=title K2Node_CustomEvent}
   > n1 = GetSoftWidgetClassForAlias!(
    execute
    self:Fort UIData Configuration Context Object Reference=Get FortUIDataConfigurationContext!() -> (ReturnValue:Fort UIData Configuration Context Object Reference) {is=title K2Node_GetBlueprintContextEx}
    WidgetAlias:"Gameplay Tag Structure (by ref)"=Get EntrySlotAliasTag() -> (EntrySlotAliasTag:Gameplay Tag Structure)
   ) -> (then, ReturnValue:User Widget Soft Class Reference)
   n2 = LazyWidgetCreated!() -> (OutputDelegate, then, Widget:User Widget Object Reference) {is=title K2Node_CustomEvent}
   SetLazyContentWithCallback!(
    execute=@n1.then
    self:Common Lazy Widget Object Reference=Get LazyWidget_EntrySlot() -> (LazyWidget_EntrySlot:Common Lazy Widget Object Reference)
    SoftWidget:User Widget Soft Class Reference=@n1
    OnCreatedCallback:"Delegate (by ref)"=@n2
   ) -> (then)
   n3 = OnStatusIndicatorLoaded!() -> (OutputDelegate, then, WidgetLoaded:User Widget Object Reference) {is=title K2Node_CustomEvent}
   > n4 = Cast To WBP_UIKit_StatusIndicator!(execute, Object=@n3.WidgetLoaded) -> (then, CastFailed, AsWBP UIKit Status Indicator:WBP UIKit Status Indicator Object Reference) {is=title K2Node_DynamicCast}
   > Set OverrideStatusIndicatorVisibility!(execute, OverrideStatusIndicatorVisibility:bool=true, self:WBP UIKit Status Indicator Base Object Reference=@n4) -> (then, Output_Get:bool)
   > Set IconSize!(execute, IconSize:E_UI_StatusIndicator_Size Enum=NewEnumerator2, self:WBP UIKit Status Indicator Base Object Reference=@n4) -> (then, Output_Get:E_UI_StatusIndicator_Size Enum=NewEnumerator2)
   > SetIndicatorType!(execute, self:WBP UIKit Status Indicator Base Object Reference=@n4, IndicatorType:E_UI_StatusIndicator_Type Enum=NewEnumerator2) -> (then)
   > SetVisibility(self=@n4, InVisibility=HitTestInvisible)
   n5 = OnPressedSelection() {is=title K2Node_CustomEvent}
   n6 = Cast To BP_PilgrimQuickplay_PartPicker_InstrumentEntryData!(
    execute=[@n5.then, @n18.then]
    Object=GetListItemObject!(UserObjectListEntry:User Object List Entry Interface) -> (ReturnValue:Object Reference)
   ) -> (
    then
    CastFailed
    AsBP Pilgrim Quickplay Part Picker Instrument Entry Data:BP Pilgrim Quickplay Part Picker Instrument Entry Data Object Reference
   ) {is=title K2Node_DynamicCast}
   > Call Part Entry on Clicked!(
    execute:Exec
    self:WBP Pilgrim Quickplay Part Picker Instrument Entry Base Object Reference
    EntryTrackType:EPilgrimTrackType Enum=Get InstrumentType!(self:BP Pilgrim Quickplay Part Picker Instrument Entry Data Object Reference=@n6) -> (InstrumentType:EPilgrimTrackType Enum=TrackGuitar)
   ) -> (then:Exec) {is=title K2Node_CallDelegate}
   n7 = Event PreConstruct()
   > n8 = GetDynamicMaterial(self=Get Image_Outline() -> (Image_Outline:Image Object Reference))
   > n9 = Set OutlineMat(OutlineMat:Material Instance Dynamic Object Reference=@n8) -> (Output_Get:Material Instance Dynamic Object Reference)
   > SetScalarParameterValue(self=@n9, ParameterName=CornerRadius, Value=0.5)
   > SetVisibility(self=Get Image_Outline(), InVisibility=Hidden)
   > SetVisibility(self=Get WBP_Sparks_BackingPanel() -> (WBP_Sparks_BackingPanel:WBP Sparks Backing Panel Object Reference), InVisibility=Hidden)
   > SetVisibility(self=Get LazyImage_RightEdge() -> (LazyImage_RightEdge:Fort Lazy Image Object Reference), InVisibility=Collapsed)
   > Branch(Condition=Get ShowEntrySlot() -> (ShowEntrySlot:bool))
   > AddWidgetAlias_PickerEntrySlot() {is=title K2Node_CallFunction}
   > Branch(Condition=@n7.IsDesignTime)
   > Set Text(Text=Get StrummingText() -> (StrummingText:Text), self=Get InfoText())
   > Branch(Condition=Get bPreview_SeeMobileAdjustments() -> (bPreview_SeeMobileAdjustments:bool))
   > SetVisibility(self=Get LazyImage_RightEdge(), InVisibility=HitTestInvisible)
   > SetBrushFromLazyMaterial!(
    execute
    self:Common Lazy Image Object Reference=Get LazyImage_RightEdge()
    LazyMaterial:"Material Interface Soft Object Reference (by ref)"=Get ExpandIconMaterial() -> (ExpandIconMaterial:Material Interface Soft Object Reference)
   ) -> (then)
   Event Construct()
   > n10 = Sequence()
   > SetVisibility(self=Get InfoContainer() -> (InfoContainer:Stack Box Object Reference), InVisibility=Collapsed)
   > n11 = Branch(Condition=Get bIsEntryOnList())
   > SetVisibility(self=Get ActionWidgetContainer() -> (ActionWidgetContainer:Overlay Object Reference), InVisibility=Collapsed)
   RemoveFromParent(execute=@n11.else, self=Get ActionWidgetContainer())
   SetMobileAdjustments(execute=@n10.then_1)
   n12 = Event OnListItemObjectSet()
   > n13 = Cast To BP_PilgrimQuickplay_PartPicker_InstrumentEntryData(Object=@n12.ListItemObject)
   > Set CurrentEntryInstrumentType(CurrentEntryInstrumentType:EPilgrimTrackType Enum=Get InstrumentType(self=@n13) -> (InstrumentType=TrackGuitar)) -> (Output_Get:EPilgrimTrackType Enum=TrackGuitar)
   > UpdateInstrumentEntry!(execute, self, EntryData:BP Pilgrim Quickplay Part Picker Instrument Entry Data Object Reference=@n13) -> (then)
   n14 = Event On Focused() {is=title K2Node_Event}
   Branch(execute=[@n14.then, @n17.then], Condition=Get bIsEntryOnList())
   > SetBrushTintColor!(
    execute
    self:Image Object Reference=Get Instrument() -> (Instrument:Image Object Reference)
    TintColor:Slate Color Structure=Make Slate Color(SpecifiedColor=(R=1,G=1,B=1,A=1))
   ) -> (then)
   > SetVisibility(self=Get ActionWidgetContainer(), InVisibility=HitTestInvisible)
   > SetVisibility(self=Get WBP_Sparks_BackingPanel(), InVisibility=HitTestInvisible)
   > SetVisibility(self=Get Image_Outline(), InVisibility=SelfHitTestInvisible)
   > DetermineShowProVocalsNotification!(execute, self, bShowTopNotification:bool=true) -> (then) {is=title K2Node_CallFunction}
   > Branch(Condition=Get bShowInfoDetail() -> (bShowInfoDetail:bool))
   > SetVisibility(self=Get InfoContainer(), InVisibility=HitTestInvisible)
   Event On Unhovered()
   > n15 = Branch(Condition=IsUsingInputType!(InputType:ECommonInputType Enum=MouseAndKeyboard, PlayerController:Player Controller Object Reference=GetOwningPlayer()) -> (ReturnValue:bool))
   Branch(execute=[@n16.then, @n15.then], Condition=Get bIsEntryOnList())
   > SetBrushTintColor(self=Get Instrument(), TintColor=Make Slate Color(SpecifiedColor=(R=0.1559,G=0.2462,B=0.8148,A=1)))
   > SetVisibility(self=Get ActionWidgetContainer(), InVisibility=Collapsed)
   > SetVisibility(self=Get WBP_Sparks_BackingPanel(), InVisibility=Hidden)
   > SetVisibility(self=Get Image_Outline(), InVisibility=Hidden)
   > DetermineShowProVocalsNotification(bShowTopNotification=false)
   > Branch(Condition=Get bShowInfoDetail())
   > SetVisibility(self=Get InfoContainer(), InVisibility=Collapsed)
   n16 = Event On Unfocused() {is=title K2Node_Event}
   Event On Hovered()
   > n17 = Branch(Condition=IsUsingInputType(InputType=MouseAndKeyboard, PlayerController=GetOwningPlayer()))
   n18 = Event On Clicked()
   n19 = Cast To BP_PilgrimQuickplay_PartPicker_InstrumentEntryData(execute=@n20.then, Object=GetListItemObject())
   > SetInstrument!(
    execute:Exec
    self:Object Reference=Get EntrySlot() -> (EntrySlot:User Widget Object Reference)
    Instrument:EPilgrimTrackType Enum=Get InstrumentType(self=@n19) -> (InstrumentType=TrackGuitar)
   ) -> (then:Exec) {is=title K2Node_Message}
   n20 = SetActiveWidget(
    execute=@n21.then
    self=Get Switcher_Info() -> (Switcher_Info:Common Visibility Switcher Object Reference)
    Widget=Get LazyWidget_EntrySlot()
   )
   n21 = Set EntrySlot(execute=@n2.then, EntrySlot:User Widget Object Reference=@n2.Widget) -> (Output_Get:User Widget Object Reference)
  ]}
 GetIntensityPipByIndex {nodes [
   n1 = GetIntensityPipByIndex!() -> (then, PipIndex:int) {is=title K2Node_FunctionEntry}
   > n2 = Branch(Condition="In Range (Integer)"!(Value:int=@n1, Min:int=0, Max:int=6, InclusiveMin:bool=true, InclusiveMax:bool=true) -> (ReturnValue:bool))
   > Return Node!(
    execute
    IntensityMark:Border Object Reference=Select!(
     Option 0:Border Object Reference=Get Intensited1() -> (Intensited1:Border Object Reference)
     Option 1:Border Object Reference=Get Intensited2() -> (Intensited2:Border Object Reference)
     Option 2:Border Object Reference=Get Intensited3() -> (Intensited3:Border Object Reference)
     Option 3:Border Object Reference=Get Intensited4() -> (Intensited4:Border Object Reference)
     Option 4:Border Object Reference=Get Intensited5() -> (Intensited5:Border Object Reference)
     Option 5:Border Object Reference=Get Intensited6() -> (Intensited6:Border Object Reference)
     Option 6:Border Object Reference=Get Intensited7() -> (Intensited7:Border Object Reference)
     Index:int=@n1
    ) -> (ReturnValue:Border Object Reference)
   )
   PrintString!(
    execute=@n2.else
    InString=Append!(A="Tried to select a part intensity pip with an index of (", B="To String (Integer)"(InInt=@n1), C:String="), out of range 0-6") -> (ReturnValue)
    bPrintToScreen
    bPrintToLog
    TextColor
    Duration
    Key=TTSAPIPWAIF83736464
   ) -> (then)
  ]}
 HideShowIcon {nodes [
   n1 = HideShowIcon!() -> (then, Show:bool=~) {is=title K2Node_FunctionEntry}
   > SetVisibility(
    self=Get Instrument()
    InVisibility=Select!(Option 0:ESlateVisibility Enum=Collapsed, Option 1:ESlateVisibility Enum=SelfHitTestInvisible, Index:bool=@n1) -> (ReturnValue:ESlateVisibility Enum)
   )
  ]}
 OnPreviewKeyDown {nodes [
   n1 = OnPreviewKeyDown!() -> (then, MyGeometry:Geometry Structure, InKeyEvent:Key Event Structure) {is=title K2Node_FunctionEntry}
   n2 = GetKey!(Input:"Key Event Structure (by ref)"=@n1.InKeyEvent) -> (ReturnValue:Key Structure)
   > n3 = Branch(Condition="Equal (Key)"(A:Key Structure=@n2, B:Key Structure=SpaceBar) -> (ReturnValue:bool=~) {is=title K2Node_PromotableOperator})
   OnPressedSelection(execute=[@n4.then, @n3.then]) {is=title K2Node_CallFunction}
   > Return Node!(execute, ReturnValue:Event Reply Structure=Handled())
   n4 = Branch(execute=@n3.else, Condition="Equal (Key)"(A=@n2, B=Enter) -> (ReturnValue=~))
   Return Node(execute=@n4.else, ReturnValue=Unhandled())
  ]}
 PartEntryOnClicked {nodes [PartEntryOnClicked!() -> (then, EntryTrackType:EPilgrimTrackType Enum=TrackGuitar) {is=title K2Node_FunctionEntry}]}
 SetIconForMicInfo {nodes [
   n1 = SetIconForMicInfo!() -> (then, Status:bool=~) {is=title K2Node_FunctionEntry}
   > n2 = Branch(Condition=@n1)
   > SetVisibility(self=Get IconPanel() -> (IconPanel:Grid Panel Object Reference), InVisibility=Collapsed)
   SetLazyContentWithCallback(
    execute=@n2.else
    self=Get IconLazyWidget() -> (IconLazyWidget:Common Lazy Widget Object Reference)
    SoftWidget=/FortUILibrary/UIKit/StatusIndicator/WBP_UIKit_StatusIndicator.WBP_UIKit_StatusIndicator_C
    OnCreatedCallback=Create Event()
   )
  ]}
 SetMobileAdjustments {nodes [
   SetMobileAdjustments() {is=title K2Node_FunctionEntry}
   > Branch(Condition=IsUsingCompactUI())
   > n1 = Branch(Condition=Get bIsEntryOnList())
   > SetMinDesiredHeight!(execute, self:Size Box Object Reference=Get SizeBox_Entry() -> (SizeBox_Entry:Size Box Object Reference), InMinDesiredHeight:float=90) -> (then)
   SetVisibility(execute=@n1.else, self=Get LazyImage_RightEdge(), InVisibility=HitTestInvisible)
   > SetBrushFromLazyMaterial(self=Get LazyImage_RightEdge(), LazyMaterial=Get ExpandIconMaterial())
   > SetMinDesiredHeight(self=Get SizeBox_Entry(), InMinDesiredHeight=75)
  ]}
 SetPeripheralPartsInfo {nodes [
   SetPeripheralPartsInfo() {is=title K2Node_FunctionEntry}
   n1 = Get PeripheralPartsWithInfo() -> (PeripheralPartsWithInfo:Array of EPilgrimTrackType Enums=TrackGuitar)
   n2 = Get CurrentEntryInstrumentType() -> (CurrentEntryInstrumentType=TrackGuitar)
   > n3 = Branch(Condition="integer >= integer"(A=Find Item!(TargetArray:Array of EPilgrimTrackType Enums=@n1, ItemToFind:"EPilgrimTrackType Enum (by ref)"=@n2) -> (ReturnValue)))
   > Set bShowInfoDetail(bShowInfoDetail=true)
   n4 = Get StrummingText()
   n5 = Get KickText() -> (KickText:Text)
   > Set Text(
    Text=Select!(
     TrackGuitar:Text
     TrackBass:Text
     TrackVocals:Text
     TrackDrum:Text
     TrackPlasticGuitar:Text=@n4
     TrackPlasticBass:Text=@n4
     TrackPlasticDrum:Text=@n5
     TrackProVocals:Text
     TrackProCymbals:Text=@n5
     TrackEvents:Text
     TrackNone:Text
     Index:EPilgrimTrackType Enum=Get CurrentEntryInstrumentType() -> (CurrentEntryInstrumentType=TrackGuitar)
    ) -> (ReturnValue:Text)
    self=Get InfoText()
   )
   > SetLazyContentWithCallback(
    self=Get IconLazyWidget()
    SoftWidget=&WBP_UIKit_StatusIndicator_C
    OnCreatedCallback=Create Event()
   )
   Set bShowInfoDetail(execute=@n3.else, bShowInfoDetail=false)
  ]}
 SetSize {nodes [
   n1 = SetSize!() -> (then, isLarge:bool=~) {is=title K2Node_FunctionEntry}
   n2 = Select!(Option 0:double=32, Option 1:double=48, Index:bool=@n1) -> (ReturnValue:double)
   > SetDesiredSizeOverride!(execute, self:Image Object Reference=Get Instrument(), DesiredSize:Vector 2D Structure=MakeVector2D(X=@n2, Y=@n2)) -> (then)
   > n3 = For Each Loop(Array:Array of Border Object References=Make Array!(
     "[0]":Border Object Reference=Get Intensited1()
     "[1]":Border Object Reference=Get Intensited2()
     "[2]":Border Object Reference=Get Intensited3()
     "[3]":Border Object Reference=Get Intensited4()
     "[4]":Border Object Reference=Get Intensited5()
     "[5]":Border Object Reference=Get Intensited6()
     "[6]":Border Object Reference=Get Intensited7()
    ) -> (Array:Array of Border Object References)) -> (Array Element:Border Object Reference)
   > Set Padding!(
    execute
    Padding:Margin Structure=Make Margin(Left=Select(Option 0=8, Option 1=12, Index=@n1), Top=Select(Option 0=20, Option 1=32, Index=@n1), Right=0, Bottom=0)
    self:Border Object Reference=@n3
   ) -> (then, Output_Get:Margin Structure)
  ]}
 UpdateInstrumentEntry {nodes [
   n1 = UpdateInstrumentEntry!() -> (then, EntryData:BP Pilgrim Quickplay Part Picker Instrument Entry Data Object Reference) {is=title K2Node_FunctionEntry}
   > n2 = Sequence!(execute) -> (then_0, then_1, then_2:Exec, then_3:Exec)
   > Set Text(
    Text=Get InstrumentName!(self:BP Pilgrim Quickplay Part Picker Instrument Entry Data Object Reference=@n1) -> (InstrumentName:Text)
    self=Get InstrumentNameLabel() -> (InstrumentNameLabel:Common Text Object Reference)
   )
   > UpdateInstrumentIcon!(
    execute
    self
    Instrument:int=Get InstrumentIcon!(self:BP Pilgrim Quickplay Part Picker Instrument Entry Data Object Reference=@n1) -> (InstrumentIcon:int=0)
   ) -> (then)
   > SetPeripheralPartsInfo() {is=title K2Node_CallFunction}
   > CheckProVocalsCase() {is=title K2Node_CallFunction}
   n3 = Branch(
    execute=@n2.then_1
    Condition=Get bIsIntensityOn!(self:BP Pilgrim Quickplay Part Picker Instrument Entry Data Object Reference=Get EntryData() -> (EntryData:BP Pilgrim Quickplay Part Picker Instrument Entry Data Object Reference)) -> (bIsIntensityOn:bool)
   )
   n4 = Get IntensityContainerDisplay() -> (IntensityContainerDisplay:Horizontal Box Object Reference)
   > Set Visibility(Visibility=SelfHitTestInvisible, self=@n4)
   > UpdateIntensityInstrument!(
    execute
    self
    Intensity:int=Get IntensityLevel!(self:BP Pilgrim Quickplay Part Picker Instrument Entry Data Object Reference=Get EntryData()) -> (IntensityLevel:int=0)
   ) -> (then)
   Set Visibility(execute=@n3.else, Visibility=Collapsed, self=@n4)
   n5 = Get!(execute:Exec=@n2.then_2) -> (then:Exec, else:Exec, EntrySlot:User Widget Object Reference)
   > SetInstrument(self=@n5, Instrument=Get CurrentEntryInstrumentType() -> (CurrentEntryInstrumentType=TrackGuitar))
   n6 = Get EntryData()
   ReceiveSongMetadataAndPart!(
    execute=@n2.then_3
    self:BP Pilgrim Quickplay Missing Parts VM Object Reference=Get BP_PilgrimQuickplay_MissingPartsVM() -> (BP_PilgrimQuickplay_MissingPartsVM:BP Pilgrim Quickplay Missing Parts VM Object Reference)
    Song Metadata:Pilgrim Song Metadata Object Reference=Get CurrentSongMetaData!(self:BP Pilgrim Quickplay Part Picker Instrument Entry Data Object Reference=@n6) -> (CurrentSongMetaData:Pilgrim Song Metadata Object Reference)
    PartSelected:EPilgrimTrackType Enum=Get InstrumentType(self=@n6) -> (InstrumentType=TrackGuitar)
   ) -> (then)
  ]}
 UpdateInstrumentIcon {nodes [
   n1 = UpdateInstrumentIcon!() -> (then, Instrument:int) {is=title K2Node_FunctionEntry}
   > n2 = CreateDynamicMaterialInstance!(execute, Parent=/PilgrimQuickplay/Widgets/Materials/M_UI_Icon_Instrument_Type_MSDF, OptionalName=None, CreationFlags=None) -> (then, ReturnValue)
   > SetScalarParameterValue(self=@n2, ParameterName=Icon, Value="To Float (Integer)"(InInt=@n1))
   > SetBrushFromMaterial(self=Get Instrument(), Material=@n2)
   n3 = Get Instrument()
   > SetBrushTintColor(self=@n3, TintColor=Make Slate Color(SpecifiedColor=(R=0.1559,G=0.2462,B=0.8148,A=1)))
   n4 = Branch(Condition=Get bIsEntryOnList())
   SetBrushTintColor(execute=@n4.else, self=@n3, TintColor=Make Slate Color(SpecifiedColor=(R=1,G=1,B=1,A=1)))
  ]}
 UpdateIntensityInstrument {nodes [
   n1 = UpdateIntensityInstrument!() -> (then, Intensity:int) {is=title K2Node_FunctionEntry}
   > Set LastPipIndexToMark(LastPipIndexToMark:int=@n1) -> (Output_Get:int=0)
   > n2 = For Loop(LastIndex=Get LastPipIndexToMark() -> (LastPipIndexToMark:int=0))
   > n3 = GetIntensityPipByIndex!(execute, self, PipIndex:int=@n2) -> (then, IntensityMark:Border Object Reference) {is=title K2Node_CallFunction}
   > SetBrushColor(self=@n3, InBrushColor=(R=1,G=1,B=1,A=1))
   n4 = For Loop(
    execute=@n2.Completed
    FirstIndex=int + int(A=Get LastPipIndexToMark() -> (LastPipIndexToMark=0))
    LastIndex=Get LastPipIndex() -> (LastPipIndex:int=0)
   )
   > n5 = GetIntensityPipByIndex(PipIndex=@n4)
   > n6 = Branch(Condition=false)
   > SetBrushColor(self=@n5, InBrushColor=(R=1,G=1,B=1,A=0.08))
   SetBrushColor(execute=@n6.else, self=@n5, InBrushColor=(R=1,G=1,B=1,A=0.2))
   Reroute Node(InputPin:Wildcard) -> (OutputPin:Wildcard)
  ]}
]
variables [
 bIsEntryOnList {type=boolean, default=true}
 BP_PilgrimPlayerControllerComponent {type=BP_PilgrimPlayerControllerComponent_C, default=null}
 bPreview_SeeMobileAdjustments {type=boolean, default=true}
 bShowInfoDetail {type=boolean}
 CurrentEntryInstrumentType {type=EPilgrimTrackType, default=TrackGuitar}
 EntrySlot {type=UserWidget, default=null}
 EntrySlotAliasTag {type=GameplayTag, default {tagName=Pilgrim.Hud.PartPicker.Entry.Slot}}
 ExpandIconMaterial {type=MaterialInterface, default=/FortUILibrary/Textures/MaterialInstances/MI_UIKit_Icon_Chevron_Up}
 FilledMaterial {type=MaterialInterface, default=/PilgrimQuickplay/Widgets/Materials/MI_UI_Pilgrim_Intensity_Filled}
 InstrumentIndex {type=integer}
 isLarge {type=boolean}
 KickText {type=string, default=Requires kick in gameplay}
 MaxIntensity {type=integer, default=7}
 OutlineMat {type=MaterialInstanceDynamic}
 PartEntryOnClicked {}
 PeripheralPartsWithInfo {type="array<EPilgrimTrackType>", default [TrackPlasticGuitar, TrackPlasticBass, TrackPlasticDrum, TrackProCymbals]}
 ShowBackground {type=boolean}
 ShowEntrySlot {type=boolean, default=true}
 ShowMeterLabel {type=boolean}
 StrummingText {type=string, default=Requires strumming in gameplay}
 UnfilledMaterial {type=MaterialInterface, default=/PilgrimQuickplay/Widgets/Materials/MI_UI_Pilgrim_Intensity_Empty}
]
widgets {
 info {inheritedWidgetCount=1, parentClass=&CommonButtonBase, rootWidgetClass=&SizeBox, widgetCount=32}
 widgets [
  SizeBox_Entry {widgetClassPath=&SizeBox, bIsVariable=true, widget=SizeBox_Entry, properties {bOverride_WidthOverride=true, widthOverride=400}}
  Overlay_34 {widgetClassPath=&Overlay, parent=SizeBox_Entry, widget=Overlay_34, slot {class=SizeBoxSlot}}
  WBP_Sparks_BackingPanel {
   widgetClassPath=/PilgrimCore/Widgets/SparksUILibrary/WBP_Sparks_BackingPanel
   bIsVariable=true
   parent=Overlay_34
   widget=WBP_Sparks_BackingPanel
   properties {blurFrame=false, cornerRadius=XLarge, darkenLayer=false, eSparksBackingPanelType=Vertical Gradient}
   slot {class=OverlaySlot, properties {horizontalAlignment=HAlign_Fill, verticalAlignment=VAlign_Fill}}
  }
  Image_Outline {
   widgetClassPath=&Image
   bIsVariable=true
   parent=Overlay_34
   widget=Image_Outline
   properties {brush {
     bIsDynamicallyLoaded=false
     drawAs=Image
     imageType=FullColor
     mirroring=NoMirror
     resourceName=None
     resourceObject=/PilgrimCore/Widgets/Materials/MI_UI_Pilgrim_Rounded_Outline_SelectedLeadeboards
     tiling=NoTile
     imageSize=<32 32>
     margin=<ltrb 0 0 0 0>
     outlineSettings {
      bUseBrushTransparency=false
      roundingType=HalfHeightRadius
      width=0
      color {colorUseRule=UseColor_Specified, specifiedColor=<rgba 0 0 0 0>}
      cornerRadii=<0 0 0 0>
     }
     tintColor {colorUseRule=UseColor_Specified, specifiedColor=<rgba 1 1 1 1>}
     uVRegion {bIsValid=false, max=<0 0>, min=<0 0>}
    }}
   slot {class=OverlaySlot, properties {horizontalAlignment=HAlign_Fill, verticalAlignment=VAlign_Fill}}
  }
  AllContainer {
   widgetClassPath=&StackBox
   parent=Overlay_34
   widget=AllContainer
   slot {class=OverlaySlot, properties {horizontalAlignment=HAlign_Fill, verticalAlignment=VAlign_Center, padding=<ltrb 16 10 16 10>}}
  }
  ActionWidgetContainer {
   widgetClassPath=&Overlay
   bIsVariable=true
   parent=AllContainer
   widget=ActionWidgetContainer
   slot {class=StackBoxSlot, properties {horizontalAlignment=HAlign_Left, verticalAlignment=VAlign_Center, padding=<ltrb -10 0 4 0>}}
  }
  CommonActionWidget {
   widgetClassPath=&CommonActionWidget
   bIsVariable=true
   parent=ActionWidgetContainer
   widget=CommonActionWidget
   properties {
    inputActions [{dataTable=&InputActionDataTable, rowName=Confirm}]
    progressMaterialBrush {
     bIsDynamicallyLoaded=false
     drawAs=Image
     imageType=NoImage
     mirroring=NoMirror
     resourceName=None
     resourceObject=null
     tiling=NoTile
     imageSize=<32 32>
     margin=<ltrb 0 0 0 0>
     outlineSettings {
      bUseBrushTransparency=false
      roundingType=HalfHeightRadius
      width=0
      color {colorUseRule=UseColor_Specified, specifiedColor=<rgba 0 0 0 0>}
      cornerRadii=<0 0 0 0>
     }
     tintColor {colorUseRule=UseColor_Specified, specifiedColor=<rgba 1 1 1 0>}
     uVRegion {bIsValid=false, max=<0 0>, min=<0 0>}
    }
   }
   slot {class=OverlaySlot}
  }
  VerticalContainer {
   widgetClassPath=&StackBox
   parent=AllContainer
   widget=VerticalContainer
   properties {orientation=Orient_Vertical}
   slot {class=StackBoxSlot, properties {size {sizeRule=Fill, value=1}}}
  }
  EntryContainer {
   widgetClassPath=&StackBox
   parent=VerticalContainer
   widget=EntryContainer
   properties {bOverride_Cursor=true}
   slot {class=StackBoxSlot}
  }
  Instrument {
   widgetClassPath=&Image
   bIsVariable=true
   parent=EntryContainer
   widget=Instrument
   properties {
    visibility=SelfHitTestInvisible
    brush {
     bIsDynamicallyLoaded=false
     drawAs=Image
     imageType=FullColor
     mirroring=NoMirror
     resourceName=None
     resourceObject=&M_UI_Icon_Instrument_Type_MSDF
     tiling=NoTile
     imageSize=<40 40>
     margin=<ltrb 0 0 0 0>
     outlineSettings {
      bUseBrushTransparency=false
      roundingType=HalfHeightRadius
      width=0
      color {colorUseRule=UseColor_Specified, specifiedColor=<rgba 0 0 0 0>}
      cornerRadii=<0 0 0 0>
     }
     tintColor {colorUseRule=UseColor_Specified, specifiedColor=<rgba 1 1 1 1>}
     uVRegion {bIsValid=false, max=<0 0>, min=<0 0>}
    }
   }
   slot {class=StackBoxSlot}
  }
  InstrumentNameLabel {
   widgetClassPath=&CommonTextBlock
   bIsVariable=true
   parent=EntryContainer
   widget=InstrumentNameLabel
   properties {
    justification=Center
    scrollStyle=/PilgrimQuickplay/Widgets/ScrollStyles/TextScrollStyle_PilgrimQuickplay_Setlist.TextScrollStyle_PilgrimQuickplay_Setlist_C
    text=Lead
    textTransformPolicy=ToUpper
    visibility=HitTestInvisible
    colorAndOpacity {colorUseRule=UseColor_Specified, specifiedColor=<rgba 0.6038 0.8879 1 1>}
    font {
     bForceMonospaced=false
     bMaterialIsStencil=false
     fontMaterial=null
     fontObject=&HeadingNow
     letterSpacing=0
     monospacedWidth=1
     size=16.5
     skewAmount=0
     typefaceFontName=76 Bold Italic
     outlineSettings {
      bApplyOutlineToDropShadows=false
      bMiteredCorners=false
      bSeparateFillAlpha=false
      outlineMaterial=null
      outlineSize=0
      outlineColor=<rgba 0 0 0 1>
     }
    }
    shadowOffset=<0 0>
   }
   slot {
    class=StackBoxSlot
    properties {horizontalAlignment=HAlign_Left, verticalAlignment=VAlign_Center, padding=<ltrb 8 0 0 0>, size {sizeRule=Fill, value=1}}
   }
  }
  Spacer_246 {widgetClassPath=&Spacer, parent=EntryContainer, widget=Spacer_246, properties {size=<10 1>}, slot {class=StackBoxSlot}}
  IconLazyWidget_LeftIntesity {
   widgetClassPath=/Script/CommonUI.CommonLazyWidget
   bIsVariable=true
   parent=EntryContainer
   widget=IconLazyWidget_LeftIntesity
   slot {class=StackBoxSlot}
  }
  Switcher_Info {
   widgetClassPath=&CommonVisibilitySwitcher
   bIsVariable=true
   parent=EntryContainer
   widget=Switcher_Info
   slot {class=StackBoxSlot, properties {horizontalAlignment=HAlign_Right}}
  }
  IntensityContainerDisplay {
   widgetClassPath=&HorizontalBox
   bIsVariable=true
   parent=Switcher_Info
   widget=IntensityContainerDisplay
   slot {class=CommonVisibilitySwitcherSlot, properties {padding=<ltrb 0 0 10 0>}}
  }
  Intensited1 {
   widgetClassPath=&Border
   bIsVariable=true
   parent=IntensityContainerDisplay
   widget=Intensited1
   properties {
    background {
     bIsDynamicallyLoaded=false
     drawAs=Image
     imageType=NoImage
     mirroring=NoMirror
     resourceName=None
     resourceObject=/PilgrimCore/Widgets/Materials/M_UI_Pilgrim_IntensityPip_Fill
     tiling=NoTile
     imageSize=<0 0>
     margin=<ltrb 0.5 0.5 0.5 0.5>
     outlineSettings {
      bUseBrushTransparency=false
      roundingType=HalfHeightRadius
      width=0
      color {colorUseRule=UseColor_Specified, specifiedColor=<rgba 0 0 0 0>}
      cornerRadii=<0 0 0 0>
     }
     tintColor {colorUseRule=UseColor_Specified, specifiedColor=<rgba 1 1 1 1>}
     uVRegion {bIsValid=false, max=<0 0>, min=<0 0>}
    }
    padding=<ltrb 8 20 0 0>
    renderTransform {angle=0, scale=<1 1>, shear=<-7 0>, translation=<0 0>}
   }
   slot {class=HorizontalBoxSlot, properties {horizontalAlignment=HAlign_Center, verticalAlignment=VAlign_Center, padding=<ltrb 3 0 0 0>}}
  }
  Intensited2 {
   widgetClassPath=&Border
   bIsVariable=true
   parent=IntensityContainerDisplay
   widget=Intensited2
   properties {
    background {
     bIsDynamicallyLoaded=false
     drawAs=Image
     imageType=NoImage
     mirroring=NoMirror
     resourceName=None
     resourceObject=&M_UI_Pilgrim_IntensityPip_Fill
     tiling=NoTile
     imageSize=<0 0>
     margin=<ltrb 0.5 0.5 0.5 0.5>
     outlineSettings {
      bUseBrushTransparency=false
      roundingType=HalfHeightRadius
      width=0
      color {colorUseRule=UseColor_Specified, specifiedColor=<rgba 0 0 0 0>}
      cornerRadii=<0 0 0 0>
     }
     tintColor {colorUseRule=UseColor_Specified, specifiedColor=<rgba 1 1 1 1>}
     uVRegion {bIsValid=false, max=<0 0>, min=<0 0>}
    }
    padding=<ltrb 8 20 0 0>
    renderTransform {angle=0, scale=<1 1>, shear=<-7 0>, translation=<0 0>}
   }
   slot {class=HorizontalBoxSlot, properties {horizontalAlignment=HAlign_Center, verticalAlignment=VAlign_Center, padding=<ltrb 3 0 0 0>}}
  }
  Intensited3 {
   widgetClassPath=&Border
   bIsVariable=true
   parent=IntensityContainerDisplay
   widget=Intensited3
   properties {
    background {
     bIsDynamicallyLoaded=false
     drawAs=Image
     imageType=NoImage
     mirroring=NoMirror
     resourceName=None
     resourceObject=&M_UI_Pilgrim_IntensityPip_Fill
     tiling=NoTile
     imageSize=<0 0>
     margin=<ltrb 0.5 0.5 0.5 0.5>
     outlineSettings {
      bUseBrushTransparency=false
      roundingType=HalfHeightRadius
      width=0
      color {colorUseRule=UseColor_Specified, specifiedColor=<rgba 0 0 0 0>}
      cornerRadii=<0 0 0 0>
     }
     tintColor {colorUseRule=UseColor_Specified, specifiedColor=<rgba 1 1 1 1>}
     uVRegion {bIsValid=false, max=<0 0>, min=<0 0>}
    }
    padding=<ltrb 8 20 0 0>
    renderTransform {angle=0, scale=<1 1>, shear=<-7 0>, translation=<0 0>}
   }
   slot {class=HorizontalBoxSlot, properties {horizontalAlignment=HAlign_Center, verticalAlignment=VAlign_Center, padding=<ltrb 3 0 0 0>}}
  }
  Intensited4 {
   widgetClassPath=&Border
   bIsVariable=true
   parent=IntensityContainerDisplay
   widget=Intensited4
   properties {
    background {
     bIsDynamicallyLoaded=false
     drawAs=Image
     imageType=NoImage
     mirroring=NoMirror
     resourceName=None
     resourceObject=&M_UI_Pilgrim_IntensityPip_Fill
     tiling=NoTile
     imageSize=<32 32>
     margin=<ltrb 0 0 0 0>
     outlineSettings {
      bUseBrushTransparency=false
      roundingType=HalfHeightRadius
      width=0
      color {colorUseRule=UseColor_Specified, specifiedColor=<rgba 0 0 0 0>}
      cornerRadii=<0 0 0 0>
     }
     tintColor {colorUseRule=UseColor_Specified, specifiedColor=<rgba 1 1 1 1>}
     uVRegion {bIsValid=false, max=<0 0>, min=<0 0>}
    }
    padding=<ltrb 8 20 0 0>
    renderTransform {angle=0, scale=<1 1>, shear=<-7 0>, translation=<0 0>}
   }
   slot {class=HorizontalBoxSlot, properties {horizontalAlignment=HAlign_Center, verticalAlignment=VAlign_Center, padding=<ltrb 3 0 0 0>}}
  }
  Intensited5 {
   widgetClassPath=&Border
   bIsVariable=true
   parent=IntensityContainerDisplay
   widget=Intensited5
   properties {
    background {
     bIsDynamicallyLoaded=false
     drawAs=Image
     imageType=NoImage
     mirroring=NoMirror
     resourceName=None
     resourceObject=&M_UI_Pilgrim_IntensityPip_Fill
     tiling=NoTile
     imageSize=<0 0>
     margin=<ltrb 0.5 0.5 0.5 0.5>
     outlineSettings {
      bUseBrushTransparency=false
      roundingType=HalfHeightRadius
      width=0
      color {colorUseRule=UseColor_Specified, specifiedColor=<rgba 0 0 0 0>}
      cornerRadii=<0 0 0 0>
     }
     tintColor {colorUseRule=UseColor_Specified, specifiedColor=<rgba 1 1 1 1>}
     uVRegion {bIsValid=false, max=<0 0>, min=<0 0>}
    }
    padding=<ltrb 8 20 0 0>
    renderTransform {angle=0, scale=<1 1>, shear=<-7 0>, translation=<0 0>}
   }
   slot {class=HorizontalBoxSlot, properties {horizontalAlignment=HAlign_Center, verticalAlignment=VAlign_Center, padding=<ltrb 3 0 0 0>}}
  }
  Intensited6 {
   widgetClassPath=&Border
   bIsVariable=true
   parent=IntensityContainerDisplay
   widget=Intensited6
   properties {
    background {
     bIsDynamicallyLoaded=false
     drawAs=Image
     imageType=NoImage
     mirroring=NoMirror
     resourceName=None
     resourceObject=&M_UI_Pilgrim_IntensityPip_Fill
     tiling=NoTile
     imageSize=<0 0>
     margin=<ltrb 0.5 0.5 0.5 0.5>
     outlineSettings {
      bUseBrushTransparency=false
      roundingType=HalfHeightRadius
      width=0
      color {colorUseRule=UseColor_Specified, specifiedColor=<rgba 0 0 0 0>}
      cornerRadii=<0 0 0 0>
     }
     tintColor {colorUseRule=UseColor_Specified, specifiedColor=<rgba 1 1 1 1>}
     uVRegion {bIsValid=false, max=<0 0>, min=<0 0>}
    }
    padding=<ltrb 8 20 0 0>
    renderTransform {angle=0, scale=<1 1>, shear=<-7 0>, translation=<0 0>}
   }
   slot {class=HorizontalBoxSlot, properties {horizontalAlignment=HAlign_Center, verticalAlignment=VAlign_Center, padding=<ltrb 3 0 0 0>}}
  }
  Intensited7 {
   widgetClassPath=&Border
   bIsVariable=true
   parent=IntensityContainerDisplay
   widget=Intensited7
   properties {
    background {
     bIsDynamicallyLoaded=false
     drawAs=Image
     imageType=NoImage
     mirroring=NoMirror
     resourceName=None
     resourceObject=&M_UI_Pilgrim_IntensityPip_Fill
     tiling=NoTile
     imageSize=<0 0>
     margin=<ltrb 0.5 0.5 0.5 0.5>
     outlineSettings {
      bUseBrushTransparency=false
      roundingType=HalfHeightRadius
      width=0
      color {colorUseRule=UseColor_Specified, specifiedColor=<rgba 0 0 0 0>}
      cornerRadii=<0 0 0 0>
     }
     tintColor {colorUseRule=UseColor_Specified, specifiedColor=<rgba 1 1 1 1>}
     uVRegion {bIsValid=false, max=<0 0>, min=<0 0>}
    }
    brushColor=<rgba 1 1 1 0.25>
    padding=<ltrb 8 20 0 0>
    renderTransform {angle=0, scale=<1 1>, shear=<-7 0>, translation=<0 0>}
   }
   slot {class=HorizontalBoxSlot, properties {horizontalAlignment=HAlign_Center, verticalAlignment=VAlign_Center, padding=<ltrb 3 0 0 0>}}
  }
  CommonTextBlock_179 {
   widgetClassPath=&CommonTextBlock
   parent=Switcher_Info
   widget=CommonTextBlock_179
   properties {
    scrollStyle=&TextScrollStyle_PilgrimQuickplay_Setlist_C
    text=Emote Only
    colorAndOpacity {colorUseRule=UseColor_Specified, specifiedColor=<rgba 1 1 1 0.7>}
    font {
     bForceMonospaced=false
     bMaterialIsStencil=false
     fontMaterial=null
     fontObject=&HeadingNow
     letterSpacing=10
     monospacedWidth=1
     size=12
     skewAmount=0
     typefaceFontName=74 Regular
     outlineSettings {
      bApplyOutlineToDropShadows=false
      bMiteredCorners=false
      bSeparateFillAlpha=false
      outlineMaterial=null
      outlineSize=0
      outlineColor=<rgba 0 0 0 1>
     }
    }
    shadowOffset=<0 0>
   }
   slot {class=CommonVisibilitySwitcherSlot, properties {horizontalAlignment=HAlign_Right, verticalAlignment=VAlign_Center}}
  }
  LazyWidget_EntrySlot {
   widgetClassPath=&CommonLazyWidget
   bIsVariable=true
   parent=Switcher_Info
   widget=LazyWidget_EntrySlot
   slot {class=CommonVisibilitySwitcherSlot}
  }
  LazyImage_RightEdge {
   widgetClassPath=&FortLazyImage
   bIsVariable=true
   parent=EntryContainer
   widget=LazyImage_RightEdge
   properties {
    brush {
     bIsDynamicallyLoaded=false
     drawAs=Image
     imageType=NoImage
     mirroring=NoMirror
     resourceName=None
     resourceObject=&MI_UIKit_Icon_Chevron_Up
     tiling=NoTile
     imageSize=<24 24>
     margin=<ltrb 0 0 0 0>
     outlineSettings {
      bUseBrushTransparency=false
      roundingType=HalfHeightRadius
      width=0
      color {colorUseRule=UseColor_Specified, specifiedColor=<rgba 0 0 0 0>}
      cornerRadii=<0 0 0 0>
     }
     tintColor {colorUseRule=UseColor_Specified, specifiedColor=<rgba 1 1 1 1>}
     uVRegion {bIsValid=false, max=<0 0>, min=<0 0>}
    }
    loadingBackgroundBrush {
     bIsDynamicallyLoaded=false
     drawAs=NoDrawType
     imageType=NoImage
     mirroring=NoMirror
     resourceName=None
     resourceObject=null
     tiling=NoTile
     imageSize=<32 17>
     margin=<ltrb 0 0 0 0>
     outlineSettings {
      bUseBrushTransparency=false
      roundingType=HalfHeightRadius
      width=0
      color {colorUseRule=UseColor_Specified, specifiedColor=<rgba 0 0 0 0>}
      cornerRadii=<0 0 0 0>
     }
     tintColor {colorUseRule=UseColor_Specified, specifiedColor=<rgba 1 1 1 1>}
     uVRegion {bIsValid=false, max=<0 0>, min=<0 0>}
    }
   }
   slot {class=StackBoxSlot, properties {horizontalAlignment=HAlign_Center, verticalAlignment=VAlign_Center}}
  }
  InfoContainer {
   widgetClassPath=&StackBox
   bIsVariable=true
   parent=VerticalContainer
   widget=InfoContainer
   slot {class=StackBoxSlot, properties {padding=<ltrb 0 5 5 0>}}
  }
  IconPanel {widgetClassPath=&GridPanel, bIsVariable=true, parent=InfoContainer, widget=IconPanel, slot {class=StackBoxSlot}}
  SizeBox_2 {
   widgetClassPath=&SizeBox
   parent=IconPanel
   widget=SizeBox_2
   properties {bOverride_HeightOverride=true, bOverride_WidthOverride=true, heightOverride=24, widthOverride=24}
   slot {class=GridSlot, properties {horizontalAlignment=HAlign_Center, verticalAlignment=VAlign_Center}}
  }
  IconLazyWidget {widgetClassPath=&CommonLazyWidget, bIsVariable=true, parent=SizeBox_2, widget=IconLazyWidget, slot {class=SizeBoxSlot}}
  Spacer_75 {widgetClassPath=&Spacer, parent=IconPanel, widget=Spacer_75, properties {size=<40 1>}, slot {class=GridSlot}}
  InfoText {
   widgetClassPath=&CommonTextBlock
   bIsVariable=true
   parent=InfoContainer
   widget=InfoText
   properties {
    scrollStyle=&TextScrollStyle_PilgrimQuickplay_Setlist_C
    colorAndOpacity {colorUseRule=UseColor_Specified, specifiedColor=<rgba 0.2705 1 1 1>}
    font {
     bForceMonospaced=false
     bMaterialIsStencil=false
     fontMaterial=null
     fontObject=&HeadingNow
     letterSpacing=0
     monospacedWidth=1
     size=12
     skewAmount=0
     typefaceFontName=74 Regular
     outlineSettings {
      bApplyOutlineToDropShadows=false
      bMiteredCorners=false
      bSeparateFillAlpha=false
      outlineMaterial=null
      outlineSize=0
      outlineColor=<rgba 0 0 0 1>
     }
    }
    shadowOffset=<0 0>
   }
   slot {
    class=StackBoxSlot
    properties {horizontalAlignment=HAlign_Left, verticalAlignment=VAlign_Center, padding=<ltrb 8 0 0 0>, size {sizeRule=Fill, value=1}}
   }
  }
  InputActionWidget {widgetClassPath=&CommonActionWidget, bInherited=true, widget=null}
 ]
}
