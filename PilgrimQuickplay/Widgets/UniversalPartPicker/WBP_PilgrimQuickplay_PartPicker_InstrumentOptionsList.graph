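# Activatable CommonUI popup listing instrument parts for the Pilgrim Quickplay part picker, with menu input and list focus.
# On activation or entry generation, it navigates to the preselected instrument until scrolling finishes, then can signal the pro vocals coachmark.
# Entry clicks and back actions play sounds and broadcast dispatchers to the owner.
# After scrolling, compact UI may cap the SizeBox height using viewport size and reported collapsed picker height.
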WidgetBlueprint /PilgrimQuickplay/Widgets/UniversalPartPicker/WBP_PilgrimQuickplay_PartPicker_InstrumentOptionsList
graphs [
 BP_GetDesiredFocusTarget {nodes [
   Get Desired Focus Target()
   > Return Node!(execute, ReturnValue:Widget Object Reference=Get CommonListView() -> (CommonListView:Common List View Object Reference))
  ]}
 BP_GetDesiredInputConfig {nodes [
   Get Desired Input Config() {is=title K2Node_FunctionEntry}
   > Return Node(ReturnValue:UIInput Config Structure=Make UIInput Config!(
     bIgnoreMoveInput:bool=true
     bIgnoreLookInput:bool=true
     InputMode:ECommonInputMode Enum=Menu
     MouseCaptureMode:EMouseCaptureMode Enum=NoCapture
     MouseLockMode:EMouseLockMode Enum=DoNotLock
     bHideCursorDuringViewportCapture:bool=True
    ) -> (UIInputConfig:UIInput Config Structure) {is=title K2Node_MakeStruct})
  ]}
 BP_OnHandleBackAction {nodes [
   On Handle Back Action() {is=title K2Node_FunctionEntry}
   > PlaySound2D(Sound=Get SoundBackAction() -> (SoundBackAction:Sound Base Object Reference))
   > Call On List Ready for Coachmark!(execute:Exec, self:WBP Pilgrim Quickplay Part Picker Instrument Options List Object Reference, bIsCoachmarkActive:bool) -> (then:Exec) {is=title K2Node_CallDelegate}
   > Call On Closing Parts List!(execute:Exec, self:WBP Pilgrim Quickplay Part Picker Instrument Options List Object Reference) -> (then:Exec) {is=title K2Node_CallDelegate}
   > Return Node(ReturnValue:bool=false)
  ]}
 CheckMobileScrollbarActivation {nodes [
   CheckMobileScrollbarActivation() {is=title K2Node_FunctionEntry}
   > Branch(Condition=IsUsingCompactUI())
   n1 = GetAbsoluteSize!(Geometry:"Geometry Structure (by ref)"=GetCachedGeometry!(self:Widget Object Reference) -> (ReturnValue:"Geometry Structure (by ref)")) -> (ReturnValue:Vector 2D Structure)
   n2 = float - float(
    A=BreakVector2D(InVec=GetViewportSize!() -> (ReturnValue:Vector 2D Structure)).Y
    B=float * float(
     A=Get CurrentUPPCollapsedHeight() -> (CurrentUPPCollapsedHeight:double)
     B=Get UPPCollapsedHeightFactor() -> (UPPCollapsedHeightFactor:double)
    ) {is=title K2Node_PromotableOperator}
   )
   > Branch(Condition="float >= float"(A=BreakVector2D(InVec=@n1).Y, B=@n2))
   n3 = float - float(A=float / float(A=@n2, B=GetViewportScale!() -> (ReturnValue:float)), B=Get SecureMarginForHeight() -> (SecureMarginForHeight:double))
   > Branch(Condition="float > float"(A=@n3))
   > SetMaxDesiredHeight!(execute, self:Size Box Object Reference=Get SizeBox_0() -> (SizeBox_0:Size Box Object Reference), InMaxDesiredHeight:float=@n3) -> (then)
  ]}
 EventGraph {nodes [
   n1 = "On Entries Generated (CommonListView)"!() -> (OutputDelegate, then, NumEntries:int=0) {is=title K2Node_ComponentBoundEvent}
   > n2 = Branch(Condition="integer > integer"(A=@n1.NumEntries, B=1))
   n3 = Branch(execute=[@n2.then, @n9.then], Condition=Get bHasNavigatedToPreselect() -> (bHasNavigatedToPreselect:bool))
   n4 = Get CommonListView()
   RequestRefresh(execute=@n3.else, self:List View Base Object Reference=@n4)
   > ForceLayoutPrepass(self:Widget Object Reference=@n4)
   > NavigateToIndex!(execute, self:List View Object Reference=@n4, Index:int=Get PreselectedInstrumentIndex() -> (PreselectedInstrumentIndex:int=0)) -> (then)
   > RequestRefreshFocus(self:Common Activatable Widget Object Reference)
   n5 = "On Entry Initialized (CommonListView)"!() -> (OutputDelegate, then, Item:Object Reference, Widget:User Widget Object Reference) {is=title K2Node_ComponentBoundEvent}
   > n6 = Cast To WBP_PilgrimQuickplay_PartPicker_InstrumentEntryBase!(execute, Object=@n5.Widget) -> (
    then
    CastFailed
    AsWBP Pilgrim Quickplay Part Picker Instrument Entry Base:WBP Pilgrim Quickplay Part Picker Instrument Entry Base Object Reference
   ) {is=title K2Node_DynamicCast}
   n7 = OnPartSelected!() -> (OutputDelegate, then, EntryTrackType:EPilgrimTrackType Enum=TrackGuitar) {is=title K2Node_CustomEvent}
   Bind Event to Part Entry on Clicked(execute=@n6.then, self:WBP Pilgrim Quickplay Part Picker Instrument Entry Base Object Reference=@n6, Delegate=@n7) {is=title K2Node_AddDelegate}
   "On List View Finished Scrolling (CommonListView)"() {is=title K2Node_ComponentBoundEvent}
   > Set bHasNavigatedToPreselect(bHasNavigatedToPreselect:bool=true) -> (Output_Get:bool)
   > CheckMobileScrollbarActivation() {is=title K2Node_CallFunction}
   n8 = GetVisibility!(self:Widget Object Reference) -> (ReturnValue:ESlateVisibility Enum=Visible)
   > Branch(Condition=AND Boolean(
     A=Get bShowingProVocalsCoachmark() -> (bShowingProVocalsCoachmark:bool)
     B="Equal (Enum)"(A:ESlateVisibility Enum=@n8, B:ESlateVisibility Enum=Visible)
    ))
   > Call On List Ready for Coachmark(bIsCoachmarkActive=true)
   OnNavigateBackIntoList() {is=title K2Node_CustomEvent}
   > Set bShowingProVocalsCoachmark(bShowingProVocalsCoachmark:bool=false) -> (Output_Get:bool)
   > NavigateToIndex(self=Get CommonListView(), Index=Get PreselectedInstrumentIndex() -> (PreselectedInstrumentIndex=0))
   Event OnRemovedFromFocusPath!() -> (OutputDelegate, then, InFocusEvent:Focus Event Structure) {is=title K2Node_Event}
   > Branch(Condition=IsPilgrimInverse!(OwningPlayerController:Player Controller Object Reference=GetOwningPlayer()) -> (IsPilgrimInverse:bool))
   > Call On Mobile Dismmised!(execute:Exec, self:WBP Pilgrim Quickplay Part Picker Instrument Options List Object Reference) -> (then:Exec) {is=title K2Node_CallDelegate}
   n9 = Event On Activated()
   Event On Deactivated()
   > Set bHasNavigatedToPreselect(bHasNavigatedToPreselect=false)
   Call On Part Entry Clicked!(
    execute:Exec=@n10.then
    self:WBP Pilgrim Quickplay Part Picker Instrument Options List Object Reference
    TrackType:EPilgrimTrackType Enum=@n7.EntryTrackType
   ) -> (then:Exec) {is=title K2Node_CallDelegate}
   n10 = Call On List Ready for Coachmark(execute=@n11.then)
   n11 = PlaySound2D(execute=@n7.then, Sound=Get SoundEntrySelected() -> (SoundEntrySelected:Sound Base Object Reference))
  ]}
 InformUPPCollapseHeight {nodes [
   n1 = InformUPPCollapseHeight!() -> (then, UPPHeight:double=~) {is=title K2Node_FunctionEntry}
   > Set CurrentUPPCollapsedHeight(CurrentUPPCollapsedHeight:double=@n1) -> (Output_Get:double)
  ]}
 On Closing Parts List {nodes [On Closing Parts List() {is=title K2Node_FunctionEntry}]}
 On List Ready For Coachmark {nodes [On List Ready For Coachmark!() -> (then, bIsCoachmarkActive:bool=~) {is=title K2Node_FunctionEntry}]}
 On Mobile Dismmised {nodes [On Mobile Dismmised() {is=title K2Node_FunctionEntry}]}
 On Part Entry Clicked {nodes [On Part Entry Clicked!() -> (then, TrackType:EPilgrimTrackType Enum=TrackGuitar) {is=title K2Node_FunctionEntry}]}
 SetCurrentlyShowingCoachmark {nodes [
   n1 = SetCurrentlyShowingCoachmark!() -> (then, CoachmarkStatus:bool=~) {is=title K2Node_FunctionEntry}
   > Set bShowingProVocalsCoachmark(bShowingProVocalsCoachmark=@n1)
  ]}
]
variables [
 bHasNavigatedToPreselect {type=boolean}
 bShowingProVocalsCoachmark {type=boolean}
 CurrentUPPCollapsedHeight {type=number}
 On Closing Parts List {}
 On List Ready For Coachmark {}
 On Mobile Dismmised {}
 On Part Entry Clicked {}
 PreselectedInstrumentIndex {type=integer}
 SecureMarginForHeight {type=number, default=70}
 SoundBackAction {type=SoundBase, default=/PilgrimQuickplay/Audio/UI/SC_PartsAndDifficulty_Close_Cue}
 SoundEntrySelected {type=SoundBase, default=/PilgrimQuickplay/Audio/UI/SC_InstrumentSelect_Cue}
 UPPCollapsedHeightFactor {type=number, default=2.15}
]
widgets {
 info {inheritedWidgetCount=0, parentClass=&CommonActivatableWidget, rootWidgetClass=&SizeBox, widgetCount=4}
 widgets [
  SizeBox_0 {widgetClassPath=&SizeBox, bIsVariable=true, widget=SizeBox_0, properties {maxDesiredHeight=600}}
  Overlay_0 {widgetClassPath=&Overlay, parent=SizeBox_0, widget=Overlay_0, slot {class=SizeBoxSlot}}
  WBP_Sparks_BackingPanel {
   widgetClassPath=/PilgrimCore/Widgets/SparksUILibrary/WBP_Sparks_BackingPanel
   bIsVariable=true
   parent=Overlay_0
   widget=WBP_Sparks_BackingPanel
   properties {blurFrame=false, cornerRadius=Small, eSparksBackingPanelType=Simple Color, visibility=HitTestInvisible}
   slot {class=OverlaySlot, properties {horizontalAlignment=HAlign_Fill, verticalAlignment=VAlign_Fill}}
  }
  CommonListView {
   widgetClassPath=&CommonListView
   bIsVariable=true
   parent=Overlay_0
   widget=CommonListView
   properties {
    bAllowKeepPreselectedItems=false
    clipping=Inherit
    entryWidgetClass=&/WBP_PilgrimQuickplay_PartPicker_InstrumentEntryBase.WBP_PilgrimQuickplay_PartPicker_InstrumentEntryBase_C
    navigation=&self:WidgetTree.CommonListView.WidgetNavigation_0
    numDesignerPreviewEntries=8
    scrollBarPadding=<ltrb 4 15 4 15>
    scrollBarStyle {
     thickness=8
     draggedThumbImage {
      bIsDynamicallyLoaded=false
      drawAs=Image
      imageType=FullColor
      mirroring=NoMirror
      resourceName=None
      resourceObject=/SparksUICommon/Materials/MI_UI_Sparks_Pbar
      tiling=NoTile
      imageSize=<0 0>
      margin=<ltrb 0.25 0.25 0.25 0.25>
      outlineSettings {
       bUseBrushTransparency=false
       roundingType=HalfHeightRadius
       width=0
       color {colorUseRule=UseColor_Specified, specifiedColor=<rgba 0 0 0 0>}
       cornerRadii=<0 0 0 0>
      }
      tintColor {colorUseRule=UseColor_Specified, specifiedColor=<rgba 1 1 1 1>}
      uVRegion {bIsValid=false, max=<0 0>, min=<0 0>}
     }
     horizontalBackgroundImage {
      bIsDynamicallyLoaded=false
      drawAs=Image
      imageType=FullColor
      mirroring=NoMirror
      resourceName=None
      resourceObject=null
      tiling=NoTile
      imageSize=<0 0>
      margin=<ltrb 0 0 0 0>
      outlineSettings {
       bUseBrushTransparency=false
       roundingType=HalfHeightRadius
       width=0
       color {colorUseRule=UseColor_Specified, specifiedColor=<rgba 0 0 0 0>}
       cornerRadii=<0 0 0 0>
      }
      tintColor {colorUseRule=UseColor_Specified, specifiedColor=<rgba 1 1 1 1>}
      uVRegion {bIsValid=false, max=<0 0>, min=<0 0>}
     }
     horizontalBottomSlotImage {
      bIsDynamicallyLoaded=false
      drawAs=Image
      imageType=FullColor
      mirroring=NoMirror
      resourceName=../../../Engine/Content/Slate/Common/Scrollbar_Background_Horizontal.png
      resourceObject=null
      tiling=NoTile
      imageSize=<8 8>
      margin=<ltrb 0 0 0 0>
      outlineSettings {
       bUseBrushTransparency=false
       roundingType=HalfHeightRadius
       width=0
       color {colorUseRule=UseColor_Specified, specifiedColor=<rgba 0 0 0 0>}
       cornerRadii=<0 0 0 0>
      }
      tintColor {colorUseRule=UseColor_Specified, specifiedColor=<rgba 1 1 1 1>}
      uVRegion {bIsValid=false, max=<0 0>, min=<0 0>}
     }
     horizontalTopSlotImage {
      bIsDynamicallyLoaded=false
      drawAs=Image
      imageType=FullColor
      mirroring=NoMirror
      resourceName=../../../Engine/Content/Slate/Common/Scrollbar_Background_Horizontal.png
      resourceObject=null
      tiling=NoTile
      imageSize=<8 8>
      margin=<ltrb 0 0 0 0>
      outlineSettings {
       bUseBrushTransparency=false
       roundingType=HalfHeightRadius
       width=0
       color {colorUseRule=UseColor_Specified, specifiedColor=<rgba 0 0 0 0>}
       cornerRadii=<0 0 0 0>
      }
      tintColor {colorUseRule=UseColor_Specified, specifiedColor=<rgba 1 1 1 1>}
      uVRegion {bIsValid=false, max=<0 0>, min=<0 0>}
     }
     hoveredThumbImage {
      bIsDynamicallyLoaded=false
      drawAs=Image
      imageType=FullColor
      mirroring=NoMirror
      resourceName=None
      resourceObject=&MI_UI_Sparks_Pbar
      tiling=NoTile
      imageSize=<0 0>
      margin=<ltrb 0.25 0.25 0.25 0.25>
      outlineSettings {
       bUseBrushTransparency=false
       roundingType=HalfHeightRadius
       width=0
       color {colorUseRule=UseColor_Specified, specifiedColor=<rgba 0 0 0 0>}
       cornerRadii=<0 0 0 0>
      }
      tintColor {colorUseRule=UseColor_Specified, specifiedColor=<rgba 1 1 1 1>}
      uVRegion {bIsValid=false, max=<0 0>, min=<0 0>}
     }
     normalThumbImage {
      bIsDynamicallyLoaded=false
      drawAs=Image
      imageType=FullColor
      mirroring=NoMirror
      resourceName=None
      resourceObject=&MI_UI_Sparks_Pbar
      tiling=NoTile
      imageSize=<0 0>
      margin=<ltrb 0.25 0.25 0.25 0.25>
      outlineSettings {
       bUseBrushTransparency=false
       roundingType=HalfHeightRadius
       width=0
       color {colorUseRule=UseColor_Specified, specifiedColor=<rgba 0 0 0 0>}
       cornerRadii=<0 0 0 0>
      }
      tintColor {colorUseRule=UseColor_Specified, specifiedColor=<rgba 1 1 1 1>}
      uVRegion {bIsValid=false, max=<0 0>, min=<0 0>}
     }
     verticalBackgroundImage {
      bIsDynamicallyLoaded=false
      drawAs=Image
      imageType=FullColor
      mirroring=NoMirror
      resourceName=None
      resourceObject=/SparksUICommon/Materials/MI_UI_Sparks_ULTimer
      tiling=NoTile
      imageSize=<0 0>
      margin=<ltrb 0 0 0 0>
      outlineSettings {
       bUseBrushTransparency=false
       roundingType=HalfHeightRadius
       width=0
       color {colorUseRule=UseColor_Specified, specifiedColor=<rgba 0 0 0 0>}
       cornerRadii=<0 0 0 0>
      }
      tintColor {colorUseRule=UseColor_Specified, specifiedColor=<rgba 1 1 1 0.6>}
      uVRegion {bIsValid=false, max=<0 0>, min=<0 0>}
     }
     verticalBottomSlotImage {
      bIsDynamicallyLoaded=false
      drawAs=Image
      imageType=FullColor
      mirroring=NoMirror
      resourceName=../../../Engine/Content/Slate/Common/Scrollbar_Background_Vertical.png
      resourceObject=null
      tiling=NoTile
      imageSize=<8 8>
      margin=<ltrb 0 0 0 0>
      outlineSettings {
       bUseBrushTransparency=false
       roundingType=HalfHeightRadius
       width=0
       color {colorUseRule=UseColor_Specified, specifiedColor=<rgba 0 0 0 0>}
       cornerRadii=<0 0 0 0>
      }
      tintColor {colorUseRule=UseColor_Specified, specifiedColor=<rgba 1 1 1 0>}
      uVRegion {bIsValid=false, max=<0 0>, min=<0 0>}
     }
     verticalTopSlotImage {
      bIsDynamicallyLoaded=false
      drawAs=Image
      imageType=FullColor
      mirroring=NoMirror
      resourceName=../../../Engine/Content/Slate/Common/Scrollbar_Background_Vertical.png
      resourceObject=null
      tiling=NoTile
      imageSize=<8 8>
      margin=<ltrb 0 0 0 0>
      outlineSettings {
       bUseBrushTransparency=false
       roundingType=HalfHeightRadius
       width=0
       color {colorUseRule=UseColor_Specified, specifiedColor=<rgba 0 0 0 0>}
       cornerRadii=<0 0 0 0>
      }
      tintColor {colorUseRule=UseColor_Specified, specifiedColor=<rgba 1 1 1 0>}
      uVRegion {bIsValid=false, max=<0 0>, min=<0 0>}
     }
    }
   }
   slot {class=OverlaySlot, properties {horizontalAlignment=HAlign_Fill, verticalAlignment=VAlign_Fill}}
  }
 ]
}
