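# Notification banner widget for the Pilgrim (Festival) Quickplay part picker, warning that the part needs microphone input or permission.
# An attention indicator and wrapped text sit over a shadowed backing, whose dynamic material alpha is set on Construct.
# SetMicrophoneIssueType picks the message by issue enum from inline localized text, not the string variables.
# SetNotificationVisibility applies the requested visibility, but forces Collapsed in compact UI.
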WidgetBlueprint /PilgrimQuickplay/Widgets/UniversalPartPicker/WBP_PilgrimQuickplay_PartPicker_TopNotification
graphs [
 EventGraph {nodes [
   Event Construct()
   > n1 = GetDynamicMaterial(self=Get Backing() -> (Backing:Image Object Reference))
   > SetScalarParameterValue(self=@n1, ParameterName=Alpha_Final, Value=Get AlphaForBackground() -> (AlphaForBackground:float))
  ]}
 SetMicrophoneIssueType {nodes [
   n1 = SetMicrophoneIssueType!() -> (then, MicrophoneIssueType:EPilgrimMicrophoneIssueType Enum=CheckMicInput) {is=title K2Node_FunctionEntry}
   > "SetText (Text)"(
    self=Get Text_NoticeBody() -> (Text_NoticeBody:Common Text Object Reference)
    InText=Select!(
     CheckMicInput:Text="LOCTEXT(\"This part requires mic input. Select mic in Settings > Game > Festival\")"
     CheckMicPermission:Text="LOCTEXT(\"This part requires system level microphone permission, you need to manually grant access in your settings app.\")"
     Unknown:Text
     Index:EPilgrimMicrophoneIssueType Enum=@n1
    ) -> (ReturnValue:Text)
   )
  ]}
 SetNotificationVisibility {nodes [
   n1 = SetNotificationVisibility!() -> (then, DesiredVisibility:ESlateVisibility Enum=Visible) {is=title K2Node_FunctionEntry}
   > n2 = Branch(Condition=IsUsingCompactUI())
   > Set Visibility!(execute, Visibility=Collapsed) -> (then, Output_Get)
   Set Visibility(execute=@n2.else, Visibility=@n1)
  ]}
]
variables [
 AlphaForBackground {type=number, default=0.8}
 DefaultProVocalsNotificationText {type=string, default="This part requires mic input. Select mic in Settings > Game > Festival"}
 Has Microphone Permission {type=boolean}
 NoMicPermissionProVocalsNotificationText {type=string, default="This part requires system level microphone permission, you need to manually grant access in your settings app."}
 SelectedInstrument {type=EPilgrimTrackType, default=TrackGuitar}
]
widgets {
 info {inheritedWidgetCount=0, parentClass=&UserWidget, rootWidgetClass=&Overlay, widgetCount=9}
 widgets [
  Notice {widgetClassPath=&Overlay, bIsVariable=true, widget=Notice, properties {visibility=HitTestInvisible}}
  BackingElements {
   widgetClassPath=&Overlay
   parent=Notice
   widget=BackingElements
   slot {class=OverlaySlot, properties {horizontalAlignment=HAlign_Fill, verticalAlignment=VAlign_Fill}}
  }
  Shadow {
   widgetClassPath=&Image
   parent=BackingElements
   widget=Shadow
   properties {
    visibility=SelfHitTestInvisible
    brush {
     bIsDynamicallyLoaded=false
     drawAs=Image
     imageType=NoImage
     mirroring=NoMirror
     resourceName=None
     resourceObject=/PilgrimCore/Widgets/Materials/M_UI_Pilgrim_RoundedShadow_Box
     tiling=NoTile
     imageSize=<32 32>
     margin=<ltrb 0 0 0 0>
     outlineSettings {
      bUseBrushTransparency=false
      roundingType=HalfHeightRadius
      width=0
      color {colorUseRule=UseColor_Specified, specifiedColor=<rgba 0 0 0 0>}
      cornerRadii=<0 0 0 0>
     }
     tintColor {colorUseRule=UseColor_Specified, specifiedColor=<rgba 1 1 1 1>}
     uVRegion {bIsValid=false, max=<0 0>, min=<0 0>}
    }
   }
   slot {class=OverlaySlot, properties {horizontalAlignment=HAlign_Fill, verticalAlignment=VAlign_Fill, padding=<ltrb -16 -16 -16 -16>}}
  }
  Backing {
   widgetClassPath=&Image
   bIsVariable=true
   parent=BackingElements
   widget=Backing
   properties {
    visibility=SelfHitTestInvisible
    brush {
     bIsDynamicallyLoaded=false
     drawAs=Image
     imageType=NoImage
     mirroring=NoMirror
     resourceName=None
     resourceObject=/PilgrimCore/Widgets/Materials/M_UI_Pilgrim_HUD_SimpleBacking
     tiling=NoTile
     imageSize=<32 32>
     margin=<ltrb 0 0 0 0>
     outlineSettings {
      bUseBrushTransparency=false
      roundingType=HalfHeightRadius
      width=0
      color {colorUseRule=UseColor_Specified, specifiedColor=<rgba 0 0 0 0>}
      cornerRadii=<0 0 0 0>
     }
     tintColor {colorUseRule=UseColor_Specified, specifiedColor=<rgba 1 1 1 1>}
     uVRegion {bIsValid=false, max=<0 0>, min=<0 0>}
    }
   }
   slot {class=OverlaySlot, properties {horizontalAlignment=HAlign_Fill, verticalAlignment=VAlign_Fill}}
  }
  Contents {
   widgetClassPath=&HorizontalBox
   parent=Notice
   widget=Contents
   slot {class=OverlaySlot, properties {horizontalAlignment=HAlign_Center, verticalAlignment=VAlign_Center, padding=<ltrb 20 20 20 20>}}
  }
  Overlay_4 {
   widgetClassPath=&Overlay
   parent=Contents
   widget=Overlay_4
   slot {class=HorizontalBoxSlot, properties {verticalAlignment=VAlign_Center}}
  }
  Image_1 {
   widgetClassPath=&Image
   parent=Overlay_4
   widget=Image_1
   properties {
    visibility=SelfHitTestInvisible
    brush {
     bIsDynamicallyLoaded=false
     drawAs=Image
     imageType=NoImage
     mirroring=NoMirror
     resourceName=None
     resourceObject=/PilgrimCore/Widgets/Materials/M_UI_StreakMeter_FX_Rings
     tiling=NoTile
     imageSize=<40 40>
     margin=<ltrb 0 0 0 0>
     outlineSettings {
      bUseBrushTransparency=false
      roundingType=HalfHeightRadius
      width=0
      color {colorUseRule=UseColor_Specified, specifiedColor=<rgba 0 0 0 0>}
      cornerRadii=<0 0 0 0>
     }
     tintColor {colorUseRule=UseColor_Specified, specifiedColor=<rgba 1 1 1 1>}
     uVRegion {bIsValid=false, max=<0 0>, min=<0 0>}
    }
    renderTransform {angle=0, scale=<2.5 2.5>, shear=<0 0>, translation=<0 0>}
   }
   slot {class=OverlaySlot}
  }
  StatusIndicator {
   widgetClassPath=/FortUILibrary/UIKit/StatusIndicator/WBP_UIKit_StatusIndicator
   bIsVariable=true
   parent=Overlay_4
   widget=StatusIndicator
   properties {type=Attention, overrideStatusIndicatorVisibility=true}
   slot {class=OverlaySlot, properties {horizontalAlignment=HAlign_Center, verticalAlignment=VAlign_Center}}
  }
  Text_NoticeBody {
   widgetClassPath=&CommonTextBlock
   bIsVariable=true
   parent=Contents
   widget=Text_NoticeBody
   properties {
    autoWrapText=true
    bIsScrollingEnabled=false
    text="This part requires mic input. Select mic in Settings > Game > Festival"
    wrapTextAt=384
    font {
     bForceMonospaced=false
     bMaterialIsStencil=false
     fontMaterial=null
     fontObject=&HeadingNow
     letterSpacing=0
     monospacedWidth=1
     size=13.5
     skewAmount=0
     typefaceFontName=74 Regular
     outlineSettings {
      bApplyOutlineToDropShadows=false
      bMiteredCorners=false
      bSeparateFillAlpha=false
      outlineMaterial=null
      outlineSize=0
      outlineColor=<rgba 0 0 0 1>
     }
    }
    shadowOffset=<0 0>
    strikeBrush {
     bIsDynamicallyLoaded=false
     drawAs=NoDrawType
     imageType=NoImage
     mirroring=NoMirror
     resourceName=None
     resourceObject=null
     tiling=NoTile
     imageSize=<32 32>
     margin=<ltrb 0 0 0 0>
     outlineSettings {
      bUseBrushTransparency=false
      roundingType=HalfHeightRadius
      width=0
      color {colorUseRule=UseColor_Specified, specifiedColor=<rgba 0 0 0 0>}
      cornerRadii=<0 0 0 0>
     }
     tintColor {colorUseRule=UseColor_Specified, specifiedColor=<rgba 1 1 1 1>}
     uVRegion {bIsValid=false, max=<0 0>, min=<0 0>}
    }
   }
   slot {class=HorizontalBoxSlot, properties {padding=<ltrb 8 0 0 0>, size {sizeRule=Fill, value=1}}}
  }
 ]
}
