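# Main controls widget for the Pilgrim Quickplay part picker, where the player chooses song part and difficulty.
# SwitchWidgetMenu toggles collapsed, parts and difficulty states, opened by clicks or a PartPickerOpen event from an async event-router listener.
# Choices go to the Pilgrim player controller component; picking a part advances to difficulty, except on compact UI without gamepad.
# It also handles a Pro Vocals coachmark, controller-disconnect fade and backstage-dependent input config.
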WidgetBlueprint /PilgrimQuickplay/Widgets/UniversalPartPicker/WBP_PilgrimQuickplay_PartPicker_MainControls
animations [
 Anim_ControllerDisconnect {
  binding_count=1
  length_seconds=0.75
  object=&self:Anim_ControllerDisconnect
  sequence {bindings [WBP_PilgrimQuickplay_PartPicker_MainControls_C_17 {
     id=#1
     possessed_class=WBP_PilgrimQuickplay_PartPicker_MainControls_C
     tracks [{
       type=MovieSceneFloatTrack
       display_name=Render Opacity
       sections [{type=MovieSceneFloatSection, end_frame=15, start_frame=0, channels [None {type=MovieSceneScriptingFloatChannel, keys [[0,1],[15,0]]}]}]
      }]
    }]}
 }
 Anim_PartsMenuExpanded {
  binding_count=3
  length_seconds=0.75
  object=&self:Anim_PartsMenuExpanded
  sequence {bindings [
    Overlay_1 {
     id=#2
     possessed_class=Overlay
     tracks [{
       type=MovieSceneByteTrack
       display_name=Clipping
       sections [{type=MovieSceneByteSection, end_frame=15, start_frame=0, channels [None {type=MovieSceneScriptingByteChannel, keys [[0,1],[15,0]]}]}]
      }]
    }
    WBP_PartPickerOptionsList {
     id=#3
     possessed_class=WBP_PilgrimQuickplay_PartPicker_InstrumentOptionsList_C
     tracks [
      {
       type=MovieScene2DTransformTrack
       sections [{
         type=MovieScene2DTransformSection
         end_frame=15
         start_frame=0
         channels [
          Translation.X {type=MovieSceneScriptingFloatChannel, const [0,1]}
          Translation.Y {type=MovieSceneScriptingFloatChannel, keys [[0,200],[15,0]]}
          Angle {type=MovieSceneScriptingFloatChannel, const [0,1]}
          Scale.X {type=MovieSceneScriptingFloatChannel, const [1,1]}
          Scale.Y {type=MovieSceneScriptingFloatChannel, const [1,1]}
          Shear.X {type=MovieSceneScriptingFloatChannel, const [0,1]}
          Shear.Y {type=MovieSceneScriptingFloatChannel, const [0,1]}
         ]
        }]
      }
      {
       type=MovieSceneFloatTrack
       display_name=Render Opacity
       sections [{
         type=MovieSceneFloatSection
         end_frame=15
         start_frame=0
         channels [None {type=MovieSceneScriptingFloatChannel, keys [[0,0.3],[10,0.3704],[15,1]]}]
        }]
      }
     ]
    }
    StackBoxSlot_8 {
     id=#4
     parent_id=#2
     possessed_class=StackBoxSlot
     tracks [{
       type=MovieSceneMarginTrack
       display_name=Padding
       sections [{
         type=MovieSceneMarginSection
         end_frame=15
         start_frame=0
         channels [
          Left {type=MovieSceneScriptingFloatChannel, const [0,1]}
          Top {type=MovieSceneScriptingFloatChannel, const [15,1]}
          Right {type=MovieSceneScriptingFloatChannel, const [0,1]}
          Bottom {type=MovieSceneScriptingFloatChannel, keys [[0,-30],[15,5]]}
         ]
        }]
      }]
    }
   ]}
 }
 Anim_DifficultyMenuExpanded {
  binding_count=3
  length_seconds=0.75
  object=&self:Anim_DifficultyMenuExpanded
  sequence {bindings [
    StackBoxSlot_8 {
     id=#5
     parent_id=#6
     possessed_class=StackBoxSlot
     tracks [{
       type=MovieSceneMarginTrack
       display_name=Padding
       sections [{
         type=MovieSceneMarginSection
         end_frame=15
         start_frame=0
         channels [
          Left {type=MovieSceneScriptingFloatChannel, const [0,1]}
          Top {type=MovieSceneScriptingFloatChannel, const [15,2]}
          Right {type=MovieSceneScriptingFloatChannel, const [0,1]}
          Bottom {type=MovieSceneScriptingFloatChannel, keys [[0,-30],[15,5]]}
         ]
        }]
      }]
    }
    WBP_DifficultyOptionsList {
     id=#7
     possessed_class=WBP_PilgrimQuickplay_PartPicker_DifficultyOptionsList_C
     tracks [
      {
       type=MovieScene2DTransformTrack
       sections [{
         type=MovieScene2DTransformSection
         end_frame=15
         start_frame=0
         channels [
          Translation.X {type=MovieSceneScriptingFloatChannel, const [0,1]}
          Translation.Y {type=MovieSceneScriptingFloatChannel, keys [[0,-200],[15,0]]}
          Angle {type=MovieSceneScriptingFloatChannel, const [0,1]}
          Scale.X {type=MovieSceneScriptingFloatChannel, const [1,1]}
          Scale.Y {type=MovieSceneScriptingFloatChannel, const [1,1]}
          Shear.X {type=MovieSceneScriptingFloatChannel, const [0,1]}
          Shear.Y {type=MovieSceneScriptingFloatChannel, const [0,1]}
         ]
        }]
      }
      {
       type=MovieSceneFloatTrack
       display_name=Render Opacity
       sections [{
         type=MovieSceneFloatSection
         end_frame=15
         start_frame=0
         channels [None {type=MovieSceneScriptingFloatChannel, keys [[0,0],[10,0.122],[15,1]]}]
        }]
      }
     ]
    }
    Overlay_1 {
     id=#6
     possessed_class=Overlay
     tracks [{
       type=MovieSceneByteTrack
       display_name=Clipping
       sections [{type=MovieSceneByteSection, end_frame=15, start_frame=0, channels [None {type=MovieSceneScriptingByteChannel, keys [[0,1],[15,0]]}]}]
      }]
    }
   ]}
 }
]
graphs [
 BP_GetDesiredInputConfig {nodes [
   Get Desired Input Config() {is=title K2Node_FunctionEntry}
   > n1 = Branch(Condition=IsAtBackstageArea!(ContextActor:Actor Object Reference=GetOwningPlayer()) -> (ReturnValue:bool))
   > Return Node!(
    execute
    ReturnValue:UIInput Config Structure=Make UIInput Config!(
     bIgnoreMoveInput:bool=False
     bIgnoreLookInput:bool=False
     InputMode:ECommonInputMode Enum=Game
     MouseCaptureMode:EMouseCaptureMode Enum=CapturePermanently
     MouseLockMode:EMouseLockMode Enum=LockOnCapture
     bHideCursorDuringViewportCapture:bool=True
    ) -> (UIInputConfig:UIInput Config Structure) {is=title K2Node_MakeStruct}
   )
   Return Node(
    execute=@n1.else
    ReturnValue=Make UIInput Config(
     bIgnoreMoveInput=true
     bIgnoreLookInput=true
     InputMode=Menu
     MouseCaptureMode=NoCapture
     MouseLockMode=DoNotLock
     bHideCursorDuringViewportCapture=True
    )
   )
  ]}
 CheckShowProVocalsCoachmark {nodes [
   CheckShowProVocalsCoachmark() {is=title K2Node_FunctionEntry}
   n1 = AND Boolean(
    A=ShouldShowOnboarding!(self:Fort Sidebar Onboard Tooltip Widget Object Reference=Get WBP_PilgrimCoachMark_ProVocals() -> (WBP_PilgrimCoachMark_ProVocals:WBP Pilgrim Coach Mark Pro Vocals Object Reference)) -> (ReturnValue:bool)
    B=Get bShowProVocalsCoachmark() -> (bShowProVocalsCoachmark:bool)
   )
   > Branch(Condition=@n1)
   > SetCurrentlyShowingCoachmark!(
    execute
    self:WBP Pilgrim Quickplay Part Picker Instrument Options List Object Reference=Get WBP_PartPickerOptionsList() -> (WBP_PartPickerOptionsList:WBP Pilgrim Quickplay Part Picker Instrument Options List Object Reference)
    CoachmarkStatus:bool=@n1
   ) -> (then)
  ]}
 EventGraph {nodes [
   "On Clicked (WBP_PilgrimQuickplay_PartPicker_ActionBar)"!() -> (OutputDelegate, then, Button:Common Button Base Object Reference) {is=title K2Node_ComponentBoundEvent}
   > On Opening To Parts()
   "On Clicked (WBP_PilgrimQuickplay_PartPickerIntensityEntry)"!() -> (OutputDelegate, then, Button:Common Button Base Object Reference) {is=title K2Node_ComponentBoundEvent}
   > Branch(Condition=IsUsingCompactUI())
   > On Opening To Parts()
   "On Clicked (WBP_PilgrimQuickplay_PartPicker_DifficultyEntryBase)"!() -> (OutputDelegate, then, Button:Common Button Base Object Reference) {is=title K2Node_ComponentBoundEvent}
   > Branch(Condition=IsUsingCompactUI())
   > On Opening To Difficulties()
   On Collapse All Selections() {is=title K2Node_CustomEvent}
   > SwitchWidgetMenu!(execute, self, CaseIndex:int=0) -> (then)
   n1 = On Instrument Selected!() -> (OutputDelegate, then, Instrument Type:EPilgrimTrackType Enum=TrackGuitar) {is=title K2Node_CustomEvent}
   > n2 = TryToGetBPPilgrimPlayerController!(execute, self) -> (then, BPPilgrimPlayerController:BP Pilgrim Player Controller Component Object Reference, bIsValid:bool)
   > n3 = Branch(Condition=@n2.bIsValid)
   > n4 = TrySwitchToInstrument!(execute, self:BP Pilgrim Player Controller Component Object Reference=@n2, Track Type:EPilgrimTrackType Enum=@n1.Instrument Type) -> (then)
   n5 = IsUsingInputType!(InputType:ECommonInputType Enum=Gamepad, PlayerController:Player Controller Object Reference=GetOwningPlayer()) -> (ReturnValue:bool)
   n6 = Branch(execute=[@n4.then, @n3.else], Condition=AND Boolean(A=NOT Boolean(A=@n5), B=IsUsingCompactUI()))
   > On Collapse All Selections() {is=title K2Node_CallFunction}
   On Opening To Difficulties(execute=@n6.else)
   On Going to Difficulties() {is=title K2Node_CustomEvent}
   > SwitchWidgetMenu(CaseIndex=2)
   On Going to Parts() {is=title K2Node_CustomEvent}
   > InformHeightForMobileScrollbar()
   > CheckShowProVocalsCoachmark() {is=title K2Node_CallFunction}
   > SwitchWidgetMenu(CaseIndex=1)
   n7 = On Difficulty Selected!() -> (OutputDelegate, then, Difficulty:EPilgrimSongDifficulty Enum=DifficultyEasy) {is=title K2Node_CustomEvent}
   > n8 = TryToGetBPPilgrimPlayerController()
   > n9 = Branch(Condition=@n8.bIsValid)
   > n10 = Try to Select Difficulty!(execute, self:BP Pilgrim Player Controller Component Object Reference=@n8, NewDifficulty:EPilgrimSongDifficulty Enum=@n7.Difficulty) -> (then)
   On Collapse All Selections(execute=[@n10.then, @n9.else])
   On Opening To Parts() {is=title K2Node_CustomEvent}
   > PlaySound2D(Sound=Get SoundListExpanded() -> (SoundListExpanded:Sound Base Object Reference))
   > On Going to Parts() {is=title K2Node_CallFunction}
   n11 = OnSwitchingProVocalsCoachmark!() -> (OutputDelegate, then, bIsCoachmarkActivated:bool=~) {is=title K2Node_CustomEvent}
   > SwitchProVocalsCoachmarkActivation!(execute, self, ActivateCoachmark:bool=@n11.bIsCoachmarkActivated) -> (then)
   On Opening To Difficulties() {is=title K2Node_CustomEvent}
   > PlaySound2D(Sound=Get SoundListExpanded())
   > On Going to Difficulties() {is=title K2Node_CallFunction}
   OnDismissedProVocalsCoachmark() {is=title K2Node_CustomEvent}
   > OnNavigateBackIntoList(self:WBP Pilgrim Quickplay Part Picker Instrument Options List Object Reference=Get WBP_PartPickerOptionsList())
   Event Destruct()
   > n12 = Sequence()
   > n13 = For Each Loop(Array:Array of Async Action Start Listening To Event Object References=Get Event Async Tasks() -> (Event Async Tasks:Array of Async Action Start Listening To Event Object References)) -> (Array Element:Async Action Start Listening To Event Object Reference)
   Clear(execute=@n13.Completed, TargetArray:Array of Async Action Start Listening To Event Object References=Get Event Async Tasks())
   n14 = Get WBP_PartPickerOptionsList()
   > Unbind all Events from On Closing Parts List!(execute:Exec, self:WBP Pilgrim Quickplay Part Picker Instrument Options List Object Reference=@n14) -> (then:Exec) {is=title K2Node_ClearDelegate}
   > Unbind all Events from On Part Entry Clicked!(execute:Exec, self:WBP Pilgrim Quickplay Part Picker Instrument Options List Object Reference=@n14) -> (then:Exec) {is=title K2Node_ClearDelegate}
   n15 = Get WBP_DifficultyOptionsList() -> (WBP_DifficultyOptionsList:WBP Pilgrim Quickplay Part Picker Difficulty Options List Object Reference)
   > Unbind all Events from On Closing Difficulty List!(execute:Exec, self:WBP Pilgrim Quickplay Part Picker Difficulty Options List Object Reference=@n15) -> (then:Exec) {is=title K2Node_ClearDelegate}
   > Unbind all Events from On Difficulty Entry Clicked!(execute:Exec, self:WBP Pilgrim Quickplay Part Picker Difficulty Options List Object Reference=@n15) -> (then:Exec) {is=title K2Node_ClearDelegate}
   > Parent: Destruct()
   StopListeningToEvent(execute=@n13.LoopBody, self:Async Action Start Listening To Event Object Reference=@n13)
   n16 = Get Wait For Playspace Async Task() -> (Wait For Playspace Async Task:Wait For Discoverable Playspace With Filter Action Object Reference)
   Is Valid(exec=@n12.then_1, InputObject=@n16)
   > Cancel(self:Wait For Discoverable Playspace With Filter Action Object Reference=@n16)
   Event Construct()
   > n17 = Sequence!(execute) -> (then_0, then_1, then_2:Exec, then_3:Exec, then_4:Exec)
   n18 = Get WBP_PartPickerOptionsList()
   > Bind Event to On Closing Parts List(self:WBP Pilgrim Quickplay Part Picker Instrument Options List Object Reference=@n18, Delegate=Create Event()) {is=title K2Node_AddDelegate}
   > Bind Event to On Part Entry Clicked(self:WBP Pilgrim Quickplay Part Picker Instrument Options List Object Reference=@n18, Delegate=Create Event()) {is=title K2Node_AddDelegate}
   > Bind Event to On Mobile Dismmised(self:WBP Pilgrim Quickplay Part Picker Instrument Options List Object Reference=@n18, Delegate=Create Event()) {is=title K2Node_AddDelegate}
   > Bind Event to On List Ready for Coachmark(self:WBP Pilgrim Quickplay Part Picker Instrument Options List Object Reference=@n18, Delegate=Create Event()) {is=title K2Node_AddDelegate}
   n19 = Get WBP_DifficultyOptionsList()
   > Bind Event to On Closing Difficulty List(self:WBP Pilgrim Quickplay Part Picker Difficulty Options List Object Reference=@n19, Delegate=Create Event()) {is=title K2Node_AddDelegate}
   > Bind Event to On Difficulty Entry Clicked(self:WBP Pilgrim Quickplay Part Picker Difficulty Options List Object Reference=@n19, Delegate=Create Event()) {is=title K2Node_AddDelegate}
   > Bind Event to On Mobile Dismissed(self:WBP Pilgrim Quickplay Part Picker Difficulty Options List Object Reference=@n19, Delegate=Create Event()) {is=title K2Node_AddDelegate}
   > Bind Event to On Pro Vocals Tooltip Dismissed(self:WBP Pilgrim Coach Mark Pro Vocals Object Reference=Get WBP_PilgrimCoachMark_ProVocals(), Delegate=Create Event()) {is=title K2Node_AddDelegate}
   n20 = PlaylistDataReadyAsync!(execute:Exec=@n17.then_1) -> (
    then:Exec
    OnReady:Exec
    GameState:Fort Game State Athena Object Reference
    Playlist:Fort Playlist Object Reference
    PlaylistContextTags:"Gameplay Tag Container Structure (by ref)"
   ) {is=title K2Node_AsyncAction}
   n21 = PilgrimWaitForEventRouter!(execute:Exec=@n20.OnReady, ContextActor:Actor Object Reference=GetOwningPlayer()) -> (then:Exec, OnEventRouterAvailable:Exec, EventRouter:Gameplay Event Router Component Object Reference) {is=title K2Node_AsyncAction}
   n22 = StartListeningToEvent!(
    execute:Exec=@n21.OnEventRouterAvailable
    Target:Gameplay Event Router Component Object Reference=@n21
    EventType:Script Struct Object Reference=/Script/PilgrimCoreRuntime.PilgrimEvent_PartPickerOpen
    Context:Object Reference
    EventBubblingRule:EEventBubblingRule Enum
   ) -> (
    then:Exec
    AsyncTaskProxy:Async Action Start Listening To Event Object Reference
    OnEventReceived:Exec
    Payload:Pilgrim Event Part Picker Open Structure
   )
   > Add(
    TargetArray:Array of Async Action Start Listening To Event Object References=Get Event Async Tasks()
    NewItem:"Async Action Start Listening To Event Object Reference (by ref)"=@n22
   ) {is=title K2Node_CallArrayFunction}
   On Opening To Parts(execute=@n22.OnEventReceived) {is=title K2Node_CallFunction}
   SetNavigationRuleBase!(execute=@n17.then_2, self:Widget Object Reference, Direction:EUINavigation Enum=Left, Rule:EUINavigationRule Enum=Stop) -> (then)
   > SetNavigationRuleBase(Direction=Right, Rule=Stop)
   n23 = Get WBP_DifficultyOptionsList()
   > SetNavigationRuleBase(self=@n23, Direction=Up, Rule=Wrap)
   > SetNavigationRuleBase(self=@n23, Direction=Down, Rule=Wrap)
   n24 = Get WBP_PartPickerOptionsList()
   > SetNavigationRuleBase(self=@n24, Direction=Up, Rule=Wrap)
   > SetNavigationRuleBase(self=@n24, Direction=Down, Rule=Wrap)
   On Collapse All Selections(execute=@n17.then_3)
   n25 = WaitForDiscoverablePlayspaceOfTypeForActor!(
    execute:Exec=@n17.then_4
    PlayspaceClass:Playspace Soft Class Reference=/PilgrimQuickplay/Playspaces/BP_PilgrimQuickplayPlayspace.BP_PilgrimQuickplayPlayspace_C
    Actor:Actor Object Reference=GetOwningPlayer()
    Requester:Object Reference
   ) -> (
    then:Exec
    AsyncTaskProxy:Wait For Discoverable Playspace With Filter Action Object Reference
    OnFoundPlayspace:Exec
    Playspace:Playspace Object Reference
   ) {is=title K2Node_AsyncAction}
   > Set Wait For Playspace Async Task(Wait For Playspace Async Task:Wait For Discoverable Playspace With Filter Action Object Reference=@n25) -> (Output_Get:Wait For Discoverable Playspace With Filter Action Object Reference)
   Branch(execute=@n25.OnFoundPlayspace, Condition=IsAtBackstageArea(ContextActor=GetOwningPlayer()))
   n26 = Self-Reference()
   > SetNavigationRuleBase(self=@n26, Direction=Up, Rule=Stop)
   > SetNavigationRuleBase(self=@n26, Direction=Down, Rule=Stop)
   n27 = Event PreConstruct()
   > Set Visibility(Visibility=Collapsed, self=Get WBP_PartPickerOptionsList())
   > Set Visibility(Visibility=Collapsed, self=Get WBP_DifficultyOptionsList())
   > Branch(Condition=@n27.IsDesignTime)
   > Set Visibility(
    Visibility=Get PartsListVisibilityforDesign() -> (PartsListVisibilityforDesign:ESlateVisibility Enum=Visible)
    self=Get WBP_PartPickerOptionsList()
   )
   > Set Visibility(
    Visibility=Get DifficultyListVisibilityforDesign() -> (DifficultyListVisibilityforDesign:ESlateVisibility Enum=Visible)
    self=Get WBP_DifficultyOptionsList()
   )
   PlayAnimationForward(InAnimation=Get Anim_PartsMenuExpanded() -> (Anim_PartsMenuExpanded:Widget Animation Object Reference), PlaybackSpeed=3.3)
   PlayAnimationForward(InAnimation=Get Anim_DifficultyMenuExpanded() -> (Anim_DifficultyMenuExpanded:Widget Animation Object Reference), PlaybackSpeed=2.4)
   n28 = Break Pilgrim Event Part Picker Open!(PilgrimEvent_PartPickerOpen:"Pilgrim Event Part Picker Open Structure (by ref)"=@n22.Payload) -> (PlayerController:Player Controller Object Reference, bIsEnhacedInputAction:bool=~) {is=title K2Node_BreakStruct}
   n29 = GetOwningPlayer()
   Branch(Condition=AND Boolean(A="Equal (Object)"(A=@n29, B=@n28), B=@n28.bIsEnhacedInputAction))
  ]}
 InformHeightForMobileScrollbar {nodes [
   InformHeightForMobileScrollbar() {is=title K2Node_FunctionEntry}
   > Branch(Condition=IsUsingCompactUI())
   n1 = GetCachedGeometry!(self:Widget Object Reference) -> (ReturnValue:"Geometry Structure (by ref)")
   > InformUPPCollapseHeight!(
    execute
    self:WBP Pilgrim Quickplay Part Picker Instrument Options List Object Reference=Get WBP_PartPickerOptionsList()
    UPPHeight:double=BreakVector2D(InVec=GetAbsoluteSize!(Geometry:"Geometry Structure (by ref)"=@n1) -> (ReturnValue:Vector 2D Structure)).Y
   ) -> (then)
  ]}
 SetActionBarVisibilityMobile {nodes [
   n1 = SetActionBarVisibilityMobile!() -> (then, ControillerConnected:bool=~) {is=title K2Node_FunctionEntry}
   > Branch(Condition=IsUsingCompactUI())
   > n2 = Branch(Condition=@n1)
   > SetVisibility(
    self=Get WBP_PilgrimQuickplay_PartPicker_ActionBar() -> (WBP_PilgrimQuickplay_PartPicker_ActionBar:WBP Pilgrim Quickplay Part Picker Action Bar Object Reference)
    InVisibility=SelfHitTestInvisible
   )
   > Set bIsActionBarVisible(bIsActionBarVisible:bool=true) -> (Output_Get:bool)
   SetVisibility(execute=@n2.else, self=Get WBP_PilgrimQuickplay_PartPicker_ActionBar(), InVisibility=Hidden)
   > Set bIsActionBarVisible(bIsActionBarVisible=false)
  ]}
 SetControllerDisconnectedVisibility {nodes [
   n1 = SetControllerDisconnectedVisibility!() -> (then, bIsControllerDisconnected:bool=~) {is=title K2Node_FunctionEntry}
   n2 = IsPlayingReplay!() -> (ReturnValue:bool)
   n3 = IsUsingCompactUI()
   > n4 = Branch(Condition=AND Boolean!(A=@n1, B=NOT Boolean(A=@n2), C:bool=NOT Boolean(A=@n3)) -> (ReturnValue))
   n5 = Get Anim_ControllerDisconnect() -> (Anim_ControllerDisconnect:Widget Animation Object Reference)
   > PlayAnimationForward(InAnimation=@n5)
   PlayAnimationReverse(execute=@n4.else, InAnimation=@n5)
  ]}
 SwitchProVocalsCoachmarkActivation {nodes [
   n1 = SwitchProVocalsCoachmarkActivation!() -> (then, ActivateCoachmark:bool=~) {is=title K2Node_FunctionEntry}
   > n2 = Branch(Condition=@n1)
   n3 = Get WBP_PilgrimCoachMark_ProVocals()
   > ActivateWidget(self=@n3)
   > RefocusOnButton(self:WBP Pilgrim Coach Mark Pro Vocals Object Reference=@n3)
   DeactivateWidget(execute=@n2.else, self=@n3)
  ]}
 SwitchUPPMainActivation {nodes [
   n1 = SwitchUPPMainActivation!() -> (then, bActivateUPP:bool=~) {is=title K2Node_FunctionEntry}
   > n2 = Branch(Condition=@n1)
   > ActivateWidget()
   DeactivateWidget(execute=@n2.else)
  ]}
 SwitchWidgetMenu {nodes [
   SwitchWidgetMenu!() -> (then, CaseIndex:int) {is=title K2Node_FunctionEntry}
   > n1 = Sequence!(execute) -> (then_0, then_1)
   > n2 = Switch on Int!(execute:Exec, Selection:int=Get CaseIndex() -> (CaseIndex:int=0)) -> (Default:Exec, 0:Exec, 1:Exec, 2:Exec)
   DeactivateWidget(execute=[@n2.0, @n2.Default], self=Get WBP_PartPickerOptionsList())
   > DeactivateWidget(self=Get WBP_DifficultyOptionsList())
   n3 = Get InstrumentInfo() -> (InstrumentInfo:Overlay Object Reference)
   > Set Visibility(Visibility=Visible, self=@n3)
   > SetRenderOpacity(self=@n3, InOpacity=1)
   n4 = Get DifficultyInfo() -> (DifficultyInfo:Overlay Object Reference)
   > Set Visibility(Visibility=Visible, self=@n4)
   > SetRenderOpacity(self=@n4, InOpacity=1)
   > n5 = Branch(Condition=Get bIsActionBarVisible() -> (bIsActionBarVisible:bool))
   > n6 = SetInfo!(
    execute
    self:WBP Pilgrim Quickplay Part Picker Action Bar Object Reference=Get WBP_PilgrimQuickplay_PartPicker_ActionBar()
    Text:Text="LOCTEXT(\"Change\")"
    Action:Data Table Row Handle Structure=Make Data Table Row Handle!(DataTable:Data Table Object Reference=/PilgrimQuickplay/Input/DT_Pilgrim_UniversalPartPicker, RowName:Name=OpenPartPickerOptions) -> (DataTableRowHandle:Data Table Row Handle Structure) {is=title K2Node_MakeStruct}
   ) -> (then)
   n7 = Get WBP_Sparks_BackingPanel_UPPCollapse() -> (WBP_Sparks_BackingPanel_UPPCollapse:WBP Sparks Backing Panel Object Reference)
   SetVisibility(execute=[@n6.then, @n5.else], self=@n7, InVisibility=HitTestInvisible)
   > SetCornerRadius!(execute, self:WBP Sparks Backing Panel Object Reference=@n7, Index:SparksBackplateCornerRadius Enum=NewEnumerator1) -> (then)
   > SetVisibility(
    self=Get WBP_Sparks_BackingPanel_Parts() -> (WBP_Sparks_BackingPanel_Parts:WBP Sparks Backing Panel Object Reference)
    InVisibility=HitTestInvisible
   )
   > SetVisibility(
    self=Get WBP_Sparks_BackingPanel_Difficulties() -> (WBP_Sparks_BackingPanel_Difficulties:WBP Sparks Backing Panel Object Reference)
    InVisibility=HitTestInvisible
   )
   > Set Text(Text=Get MainHeaderText() -> (MainHeaderText:Text), self=Get HeaderText() -> (HeaderText:Common Text Object Reference))
   ActivateWidget(execute=@n2.1, self=Get WBP_PartPickerOptionsList())
   > DeactivateWidget(self=Get WBP_DifficultyOptionsList())
   > Set Visibility(Visibility=Collapsed, self=Get InstrumentInfo())
   n8 = Get DifficultyInfo()
   > Set Visibility(Visibility=Visible, self=@n8)
   > SetRenderOpacity(self=@n8, InOpacity=0.6)
   > n9 = Branch(Condition=Get bIsActionBarVisible())
   > n10 = SetInfo(
    self=Get WBP_PilgrimQuickplay_PartPicker_ActionBar()
    Text="LOCTEXT(\"Close\")"
    Action=Make Data Table Row Handle(DataTable=&InputActionDataTable, RowName=Cancel)
   )
   n11 = Get WBP_Sparks_BackingPanel_UPPCollapse()
   SetVisibility(execute=[@n10.then, @n9.else], self=@n11, InVisibility=HitTestInvisible)
   > SetCornerRadius(self=@n11, Index=NewEnumerator0)
   > SetVisibility(self=Get WBP_Sparks_BackingPanel_Parts(), InVisibility=Hidden)
   > SetVisibility(self=Get WBP_Sparks_BackingPanel_Difficulties(), InVisibility=HitTestInvisible)
   > Set Text(Text=Get PartHeaderText() -> (PartHeaderText:Text), self=Get HeaderText())
   ActivateWidget(execute=@n2.2, self=Get WBP_DifficultyOptionsList())
   > DeactivateWidget(self=Get WBP_PartPickerOptionsList())
   n12 = Get InstrumentInfo()
   > Set Visibility(Visibility=Visible, self=@n12)
   > SetRenderOpacity(self=@n12, InOpacity=0.6)
   > Set Visibility(Visibility=Collapsed, self=Get DifficultyInfo())
   > n13 = Branch(Condition=Get bIsActionBarVisible())
   > n14 = SetInfo(
    self=Get WBP_PilgrimQuickplay_PartPicker_ActionBar()
    Text="LOCTEXT(\"Back\")"
    Action=Make Data Table Row Handle(DataTable=&InputActionDataTable, RowName=Cancel)
   )
   n15 = Get WBP_Sparks_BackingPanel_UPPCollapse()
   SetVisibility(execute=[@n14.then, @n13.else], self=@n15, InVisibility=HitTestInvisible)
   > SetCornerRadius(self=@n15, Index=NewEnumerator0)
   > SetVisibility(self=Get WBP_Sparks_BackingPanel_Parts(), InVisibility=HitTestInvisible)
   > SetVisibility(self=Get WBP_Sparks_BackingPanel_Difficulties(), InVisibility=Hidden)
   > Set Text(Text=Get DifficultyHeaderText() -> (DifficultyHeaderText:Text), self=Get HeaderText())
   RequestRefreshFocus(execute=@n1.then_1, self:Common Activatable Widget Object Reference)
  ]}
 TryToGetBPPilgrimPlayerController {nodes [
   TryToGetBPPilgrimPlayerController() {is=title K2Node_FunctionEntry}
   n1 = GetOwningPlayer()
   > n2 = Is Valid(InputObject=@n1)
   Return Node!(
    execute=[@n2.Is Not Valid, @n4.Is Not Valid, @n6.Is Not Valid, @n8.Is Not Valid]
    BPPilgrimPlayerController:BP Pilgrim Player Controller Component Object Reference
    bIsValid:bool=false
   )
   n3 = Get PlayerState!(self:Controller Object Reference=@n1) -> (PlayerState:Player State Object Reference)
   > n4 = Is Valid(InputObject=@n3)
   n5 = GetPlayerController!(self:Player State Object Reference=@n3) -> (ReturnValue:Player Controller Object Reference)
   > n6 = Is Valid(InputObject=@n5)
   n7 = GetComponentByClass!(
    self:Actor Object Reference=@n5
    ComponentClass:Actor Component Class Reference=/PilgrimCore/Player/Components/BP_PilgrimPlayerControllerComponent.BP_PilgrimPlayerControllerComponent_C
   ) -> (ReturnValue:BP Pilgrim Player Controller Component Object Reference)
   > n8 = Is Valid(InputObject=@n7)
   > Return Node(BPPilgrimPlayerController=@n7, bIsValid=true)
  ]}
]
variables [
 _RepresentedPlayer {type=Pawn, default=null}
 bIsActionBarVisible {type=boolean, default=true}
 bShowProVocalsCoachmark {type=boolean}
 DifficultyHeaderText {type=string, default=Choose Difficulty}
 DifficultyListVisibilityforDesign {type=ESlateVisibility, default=Collapsed}
 Event Async Tasks {type="array<AsyncAction_StartListeningToEvent>", default []}
 Event Router {type=GameplayEventRouterComponent, default=null}
 LocalBandState {type=PilgrimQuickplayBandStateComponent, default=null}
 MainHeaderText {type=string, default=Part + Difficulty}
 PartHeaderText {type=string, default=Choose Song Part}
 PartsListVisibilityforDesign {type=ESlateVisibility, default=Collapsed}
 SelectedInstrument {type=EPilgrimTrackType, default=TrackGuitar}
 SelectedSongMetaData {type=PilgrimSongMetadata, default=null}
 SoundListExpanded {type=SoundBase, default=/PilgrimQuickplay/Audio/UI/SC_PartsAndDifficulty_Open_Cue}
 Wait For Playspace Async Task {type=WaitForDiscoverablePlayspaceWithFilterAction, default=null}
]
widgets {
 info {
  inheritedWidgetCount=0
  parentClass=/Script/PilgrimQuickplayUI.PilgrimPartPickerBaseWidget
  rootWidgetClass=&Overlay
  widgetCount=20
 }
 widgets [
  Overlay_0 {widgetClassPath=&Overlay, widget=Overlay_0}
  SizeBox_0 {
   widgetClassPath=&SizeBox
   parent=Overlay_0
   widget=SizeBox_0
   properties {bOverride_MinDesiredWidth=true, bOverride_WidthOverride=true, minDesiredWidth=400, widthOverride=400}
   slot {class=OverlaySlot, properties {horizontalAlignment=HAlign_Fill, verticalAlignment=VAlign_Bottom}}
  }
  StackBox_21 {
   widgetClassPath=&StackBox
   parent=SizeBox_0
   widget=StackBox_21
   properties {orientation=Orient_Vertical}
   slot {class=SizeBoxSlot, properties {verticalAlignment=VAlign_Bottom}}
  }
  Overlay_1 {widgetClassPath=&Overlay, parent=StackBox_21, widget=Overlay_1, slot {class=StackBoxSlot, properties {padding=<ltrb 0 15 0 5>}}}
  WBP_Sparks_BackingPanel_UPPCollapse {
   widgetClassPath=/PilgrimCore/Widgets/SparksUILibrary/WBP_Sparks_BackingPanel
   bIsVariable=true
   parent=Overlay_1
   widget=WBP_Sparks_BackingPanel_UPPCollapse
   properties {visibility=HitTestInvisible}
   slot {class=OverlaySlot, properties {horizontalAlignment=HAlign_Fill, verticalAlignment=VAlign_Fill}}
  }
  StackBox_0 {
   widgetClassPath=&StackBox
   parent=Overlay_1
   widget=StackBox_0
   properties {orientation=Orient_Vertical}
   slot {class=OverlaySlot, properties {horizontalAlignment=HAlign_Fill, verticalAlignment=VAlign_Bottom, padding=<ltrb 0 0 0 10>}}
  }
  HeaderText {
   widgetClassPath=&CommonTextBlock
   bIsVariable=true
   parent=StackBox_0
   widget=HeaderText
   properties {
    scrollStyle=/PilgrimQuickplay/Widgets/ScrollStyles/TextScrollStyle_PilgrimQuickplay_Setlist.TextScrollStyle_PilgrimQuickplay_Setlist_C
    text=Part + Difficulty
    font {
     bForceMonospaced=false
     bMaterialIsStencil=false
     fontMaterial=null
     fontObject=&HeadingNow
     letterSpacing=0
     monospacedWidth=1
     size=18
     skewAmount=0
     typefaceFontName=75 Medium
     outlineSettings {
      bApplyOutlineToDropShadows=false
      bMiteredCorners=false
      bSeparateFillAlpha=false
      outlineMaterial=null
      outlineSize=0
      outlineColor=<rgba 0 0 0 1>
     }
    }
    shadowOffset=<0 0>
   }
   slot {class=StackBoxSlot, properties {verticalAlignment=VAlign_Center, padding=<ltrb 26 10 26 10>}}
  }
  InstrumentInfo {
   widgetClassPath=&Overlay
   bIsVariable=true
   parent=StackBox_0
   widget=InstrumentInfo
   slot {class=StackBoxSlot, properties {padding=<ltrb 10 0 10 2.5>}}
  }
  WBP_Sparks_BackingPanel_Parts {
   widgetClassPath=&WBP_Sparks_BackingPanel
   bIsVariable=true
   parent=InstrumentInfo
   widget=WBP_Sparks_BackingPanel_Parts
   properties {cornerRadius=XLarge, eSparksBackingPanelType=Simple Color, visibility=HitTestInvisible}
   slot {class=OverlaySlot, properties {horizontalAlignment=HAlign_Fill, verticalAlignment=VAlign_Fill}}
  }
  WBP_PilgrimQuickplay_PartPickerIntensityEntry {
   widgetClassPath=&/WBP_PilgrimQuickplay_PartPicker_InstrumentEntryBase
   bIsVariable=true
   parent=InstrumentInfo
   widget=WBP_PilgrimQuickplay_PartPickerIntensityEntry
   properties {bDisplayInputActionWhenNotInteractable=false, bIsEntryOnList=false, bIsFocusable=false, hoveredSlateSoundOverride {resourceObject=null}}
   slot {class=OverlaySlot, properties {horizontalAlignment=HAlign_Fill, verticalAlignment=VAlign_Fill}}
  }
  WBP_PartPickerOptionsList {
   widgetClassPath=&/WBP_PilgrimQuickplay_PartPicker_InstrumentOptionsList
   bIsVariable=true
   parent=StackBox_0
   widget=WBP_PartPickerOptionsList
   properties {
    actionDomainOverride=/Game/UI/ActionDomain/ActionDomain_Menu
    bIsBackActionDisplayedInActionBar=false
    bOverrideActionDomain=true
    desiredFocusWidget {widgetName=None}
   }
   slot {class=StackBoxSlot, properties {padding=<ltrb 10 0 10 10>}}
  }
  WBP_DifficultyOptionsList {
   widgetClassPath=&/WBP_PilgrimQuickplay_PartPicker_DifficultyOptionsList
   bIsVariable=true
   parent=StackBox_0
   widget=WBP_DifficultyOptionsList
   properties {actionDomainOverride=&ActionDomain_Menu, bIsBackActionDisplayedInActionBar=false, bOverrideActionDomain=true}
   slot {class=StackBoxSlot, properties {padding=<ltrb 10 10 10 0>}}
  }
  DifficultyInfo {
   widgetClassPath=&Overlay
   bIsVariable=true
   parent=StackBox_0
   widget=DifficultyInfo
   slot {class=StackBoxSlot, properties {padding=<ltrb 10 2.5 10 0>}}
  }
  WBP_Sparks_BackingPanel_Difficulties {
   widgetClassPath=&WBP_Sparks_BackingPanel
   bIsVariable=true
   parent=DifficultyInfo
   widget=WBP_Sparks_BackingPanel_Difficulties
   properties {cornerRadius=XLarge, eSparksBackingPanelType=Simple Color, visibility=HitTestInvisible}
   slot {class=OverlaySlot, properties {horizontalAlignment=HAlign_Fill, verticalAlignment=VAlign_Fill}}
  }
  WBP_PilgrimQuickplay_PartPicker_DifficultyEntryBase {
   widgetClassPath=&/WBP_PilgrimQuickplay_PartPicker_DifficultyEntryBase
   bIsVariable=true
   parent=DifficultyInfo
   widget=WBP_PilgrimQuickplay_PartPicker_DifficultyEntryBase
   properties {bIsEntryOnList=false, bIsFocusable=false, hoveredSlateSoundOverride {resourceObject=null}}
   slot {class=OverlaySlot, properties {horizontalAlignment=HAlign_Fill, verticalAlignment=VAlign_Fill}}
  }
  WBP_PilgrimQuickplay_PartPicker_ActionBar {
   widgetClassPath=&/WBP_PilgrimQuickplay_PartPicker_ActionBar
   bIsVariable=true
   parent=StackBox_21
   widget=WBP_PilgrimQuickplay_PartPicker_ActionBar
   properties {bIsFocusable=false}
   slot {class=StackBoxSlot, properties {horizontalAlignment=HAlign_Center}}
  }
  CommonVisualAttachment_116 {
   widgetClassPath=&CommonVisualAttachment
   parent=Overlay_0
   widget=CommonVisualAttachment_116
   properties {contentAnchor=<-0.03 -0.1>}
   slot {class=OverlaySlot, properties {horizontalAlignment=HAlign_Right}}
  }
  WBP_PilgrimCoachMark_ProVocals {
   widgetClassPath=/PilgrimQuickplay/Widgets/Coachmarks/WBP_PilgrimCoachMark_ProVocals
   bIsVariable=true
   parent=CommonVisualAttachment_116
   widget=WBP_PilgrimCoachMark_ProVocals
   properties {navigation=&self:WidgetTree.WBP_PilgrimCoachMark_ProVocals.WidgetNavigation_0}
   slot {class=SizeBoxSlot}
  }
  CommonVisualAttachment_271 {
   widgetClassPath=&CommonVisualAttachment
   parent=Overlay_0
   widget=CommonVisualAttachment_271
   properties {bOverride_WidthOverride=true, widthOverride=400, contentAnchor=<0 1>}
   slot {class=OverlaySlot}
  }
  WBP_PilgrimQuickplay_PartPicker_TopNotification {
   widgetClassPath=&/WBP_PilgrimQuickplay_PartPicker_TopNotification
   bIsVariable=true
   parent=CommonVisualAttachment_271
   widget=WBP_PilgrimQuickplay_PartPicker_TopNotification
   properties {alphaForBackground=0.7}
   slot {class=SizeBoxSlot, properties {horizontalAlignment=HAlign_Left, verticalAlignment=VAlign_Center, padding=<ltrb 0 0 0 -8>}}
  }
 ]
}
